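# Unlit translucent Niagara material, usable on meshes, ribbons and sprites, for the outer shell of a tentacle circular beam attack.
# A panning edge texture, mirrored across V, yields refined masks for opacity and for blending a dark tint with a bright, plasma-modulated color.
# A dynamic Spawn parameter drives a wipe along U, multiplied with the edge mask and particle alpha.
# UseAsSprite blends toward rotated UVs; Offset has no effect because both lerp inputs match.
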Material /PistolCanary/Assets/FX/TentacleCircularBeamAttack/M_Beam_Shell
graph {
 expressions {
  MaterialExpressionAdd_0 {
   class=MaterialExpressionAdd
   inputs [
    A {from=MaterialExpressionOneMinus_1}
    B {from=MaterialExpressionDynamicParameter_1, output=Spawn}
   ]
  }
  MaterialExpressionConstant2Vector_2 {class=MaterialExpressionConstant2Vector, properties {g=1}}
  MaterialExpressionConvert_0 {
   class=MaterialExpressionConvert
   inputs [
    None {from=MaterialExpressionMaterialFunctionCall_3, output=R}
    None {from=MaterialExpressionLinearInterpolate_0}
   ]
   properties {
    nodeName=Make Float2
    convertInputs [
     {
      type=Scalar
      expressionInput {
       expression=&self:MaterialExpressionMaterialFunctionCall_3
       inputName=None
       mask=0
       maskA=0
       maskB=0
       maskG=0
       maskR=0
       outputIndex=0
      }
     }
     {
      type=Scalar
      expressionInput {
       expression=&self:MaterialExpressionLinearInterpolate_0
       inputName=None
       mask=0
       maskA=0
       maskB=0
       maskG=0
       maskR=0
       outputIndex=0
      }
     }
    ]
    convertOutputs [{type=Vector2}]
   }
  }
  MaterialExpressionConvert_1 {
   class=MaterialExpressionConvert
   inputs [
    None {from=MaterialExpressionOneMinus_2}
    None {from=MaterialExpressionMaterialFunctionCall_9, output=R}
   ]
   properties {
    nodeName=Make Float2
    convertInputs [
     {
      type=Scalar
      expressionInput {
       expression=&self:MaterialExpressionOneMinus_2
       inputName=None
       mask=0
       maskA=0
       maskB=0
       maskG=0
       maskR=0
       outputIndex=0
      }
     }
     {
      type=Scalar
      expressionInput {
       expression=&self:MaterialExpressionMaterialFunctionCall_9
       inputName=None
       mask=0
       maskA=0
       maskB=0
       maskG=0
       maskR=0
       outputIndex=0
      }
     }
    ]
    convertOutputs [{type=Vector2}]
   }
  }
  MaterialExpressionDivide_0 {
   class=MaterialExpressionDivide
   inputs [
    A {from=MaterialExpressionLinearInterpolate_2}
    B {from=""}
   ]
   properties {constB=3.5}
  }
  MaterialExpressionDynamicParameter_0 {class=MaterialExpressionDynamicParameter, inputs [], properties {parameterIndex=3, paramNames [Offset,Param2,Param3,Param4]}}
  MaterialExpressionDynamicParameter_1 {class=MaterialExpressionDynamicParameter, inputs [], properties {parameterIndex=2, paramNames [Spawn,Param2,Param3,Param4]}}
  MaterialExpressionFloor_0 {class=MaterialExpressionFloor, inputs [None {from=MaterialExpressionMultiply_1, output=""}], properties {}}
  MaterialExpressionFrac_0 {class=MaterialExpressionFrac, inputs [None {from=MaterialExpressionMultiply_1, output=""}], properties {}}
  MaterialExpressionLinearInterpolate_0 {
   class=MaterialExpressionLinearInterpolate
   inputs [
    A {from=MaterialExpressionOneMinus_0}
    B {from=MaterialExpressionFrac_0}
    Alpha {from=MaterialExpressionFloor_0}
   ]
  }
  MaterialExpressionLinearInterpolate_1 {
   class=MaterialExpressionLinearInterpolate
   inputs [
    A {from=MaterialExpressionVectorParameter_4, output=R}
    B {from=MaterialExpressionVectorParameter_4, output=R}
    Alpha {from=MaterialExpressionDynamicParameter_0, output=Offset}
   ]
  }
  MaterialExpressionLinearInterpolate_2 {
   class=MaterialExpressionLinearInterpolate
   inputs [
    A {from=MaterialExpressionVectorParameter_6, output=RGB}
    B {from=MaterialExpressionMultiply_12}
    Alpha {from=MaterialExpressionMaterialFunctionCall_6, output=Result}
   ]
  }
  MaterialExpressionLinearInterpolate_3 {
   class=MaterialExpressionLinearInterpolate
   inputs [
    A {from=MaterialExpressionTextureCoordinate_13}
    B {from=MaterialExpressionConvert_1}
    Alpha {from=MaterialExpressionScalarParameter_0}
   ]
  }
  MaterialExpressionMaterialFunctionCall_0 {
   class=MaterialExpressionMaterialFunctionCall
   inputs [
    Texture {from=MaterialExpressionTextureObjectParameter_0}
    AdditionalUVOffset {from=MaterialExpressionLinearInterpolate_1}
    UVControl_RGB {from=MaterialExpressionVectorParameter_1}
    UVControl_A {from=MaterialExpressionVectorParameter_1}
    "[Override]TimeInput" {from=""}
    "[Override]ScaleUVByCenter" {from=MaterialExpressionStaticBool_0}
    "[Override]UVInput" {from=MaterialExpressionConvert_0}
   ]
   properties {materialFunction=/Game/Effects/Niagara/MasterMaterials/Functions/Common/FXCommon_TextureSample_V2}
  }
  MaterialExpressionMaterialFunctionCall_1 {
   class=MaterialExpressionMaterialFunctionCall
   inputs [Float2 {from=MaterialExpressionLinearInterpolate_3}]
   properties {materialFunction=&BreakOutFloat2Components}
  }
  MaterialExpressionMaterialFunctionCall_2 {
   class=MaterialExpressionMaterialFunctionCall
   inputs [
    Texture {from=MaterialExpressionMaterialFunctionCall_0, output=R}
    RGB {from=MaterialExpressionVectorParameter_2, output=RGB}
    A {from=MaterialExpressionVectorParameter_2, output=RGB}
   ]
   properties {materialFunction=&MF_FX_Refine}
  }
  MaterialExpressionMaterialFunctionCall_3 {
   class=MaterialExpressionMaterialFunctionCall
   inputs [Float2 {from=MaterialExpressionLinearInterpolate_3}]
   properties {materialFunction=&BreakOutFloat2Components}
  }
  MaterialExpressionMaterialFunctionCall_4 {
   class=MaterialExpressionMaterialFunctionCall
   inputs [Emissive {from=MaterialExpressionDivide_0}]
   properties {materialFunction=&MF_FX_EmissiveAdjustments}
  }
  MaterialExpressionMaterialFunctionCall_5 {
   class=MaterialExpressionMaterialFunctionCall
   inputs [
    TexCoord {from=MaterialExpressionMaterialFunctionCall_0, output=UV}
    ClampChannel {from=MaterialExpressionConstant2Vector_2}
   ]
   properties {materialFunction=/Game/Effects/Fort_Effects/MaterialFunctions/MF_FX_TilingMethod}
  }
  MaterialExpressionMaterialFunctionCall_6 {
   class=MaterialExpressionMaterialFunctionCall
   inputs [
    Texture {from=MaterialExpressionMaterialFunctionCall_0, output=R}
    RGB {from=MaterialExpressionVectorParameter_5, output=RGB}
    A {from=MaterialExpressionVectorParameter_5, output=RGB}
   ]
   properties {materialFunction=&MF_FX_Refine}
  }
  MaterialExpressionMaterialFunctionCall_7 {
   class=MaterialExpressionMaterialFunctionCall
   inputs [
    Texture {from=MaterialExpressionTextureObjectParameter_1}
    AdditionalUVOffset {from=""}
    UVControl_RGB {from=MaterialExpressionVectorParameter_7}
    UVControl_A {from=MaterialExpressionVectorParameter_7}
    "[Override]TimeInput" {from=""}
    "[Override]UVInput" {from=MaterialExpressionLinearInterpolate_3}
   ]
   properties {materialFunction=&FXCommon_TextureSample}
  }
  MaterialExpressionMaterialFunctionCall_8 {
   class=MaterialExpressionMaterialFunctionCall
   inputs [
    Texture {from=MaterialExpressionMaterialFunctionCall_7, output=R}
    RGB {from=MaterialExpressionVectorParameter_8, output=RGB}
    A {from=MaterialExpressionVectorParameter_8, output=RGB}
   ]
   properties {materialFunction=&MF_FX_Refine}
  }
  MaterialExpressionMaterialFunctionCall_9 {
   class=MaterialExpressionMaterialFunctionCall
   inputs [Float2 {from=MaterialExpressionTextureCoordinate_13}]
   properties {materialFunction=&BreakOutFloat2Components}
  }
  MaterialExpressionMultiply_0 {
   class=MaterialExpressionMultiply
   inputs [
    A {from=MaterialExpressionMultiply_11}
    B {from=MaterialExpressionSaturate_3}
   ]
  }
  MaterialExpressionMultiply_1 {
   class=MaterialExpressionMultiply
   inputs [
    A {from=MaterialExpressionMaterialFunctionCall_3, output=G}
    B {from=""}
   ]
   properties {constB=2}
  }
  MaterialExpressionMultiply_10 {
   class=MaterialExpressionMultiply
   inputs [
    A {from=MaterialExpressionAdd_0}
    B {from=MaterialExpressionScalarParameter_7}
   ]
  }
  MaterialExpressionMultiply_11 {
   class=MaterialExpressionMultiply
   inputs [
    A {from=MaterialExpressionMaterialFunctionCall_2, output=Result}
    B {from=MaterialExpressionMaterialFunctionCall_5, output=Result}
   ]
  }
  MaterialExpressionMultiply_12 {
   class=MaterialExpressionMultiply
   inputs [
    A {from=MaterialExpressionMaterialFunctionCall_8, output=Result}
    B {from=MaterialExpressionVectorParameter_3, output=RGB}
   ]
  }
  MaterialExpressionMultiply_2 {
   class=MaterialExpressionMultiply
   inputs [
    A {from=MaterialExpressionMultiply_0}
    B {from=MaterialExpressionParticleColor_0, output=A}
   ]
  }
  MaterialExpressionOneMinus_0 {class=MaterialExpressionOneMinus, inputs [None {from=MaterialExpressionFrac_0}]}
  MaterialExpressionOneMinus_1 {class=MaterialExpressionOneMinus, inputs [None {from=MaterialExpressionMaterialFunctionCall_1, output=R}]}
  MaterialExpressionOneMinus_2 {class=MaterialExpressionOneMinus, inputs [None {from=MaterialExpressionMaterialFunctionCall_9, output=G}]}
  MaterialExpressionParticleColor_0 {class=MaterialExpressionParticleColor, inputs [], properties {}}
  MaterialExpressionPower_6 {
   class=MaterialExpressionPower
   inputs [
    Base {from=MaterialExpressionSaturate_2}
    Exp {from=MaterialExpressionScalarParameter_8}
   ]
  }
  MaterialExpressionSaturate_2 {class=MaterialExpressionSaturate, inputs [None {from=MaterialExpressionMultiply_10}]}
  MaterialExpressionSaturate_3 {class=MaterialExpressionSaturate, inputs [None {from=MaterialExpressionPower_6}]}
  MaterialExpressionScalarParameter_0 {class=MaterialExpressionScalarParameter, properties {parameterName=UseAsSprite}}
  MaterialExpressionScalarParameter_7 {class=MaterialExpressionScalarParameter, properties {defaultValue=1, group=Spawn, parameterName=Spawn_GrayAdjust_Mul}}
  MaterialExpressionScalarParameter_8 {class=MaterialExpressionScalarParameter, properties {defaultValue=2, group=Spawn, parameterName=Spawn_GrayAdjust_Power}}
  MaterialExpressionStaticBool_0 {class=MaterialExpressionStaticBool}
  MaterialExpressionTextureCoordinate_0 {class=MaterialExpressionTextureCoordinate}
  MaterialExpressionTextureCoordinate_13 {class=MaterialExpressionTextureCoordinate}
  MaterialExpressionTextureCoordinate_2 {class=MaterialExpressionTextureCoordinate}
  MaterialExpressionTextureObjectParameter_0 {
   class=MaterialExpressionTextureObjectParameter
   inputs [
    UVs {from="", output=""}
    Apply View MipBias {from="", output=""}
   ]
   properties {
    group=Edge
    parameterName=Edge_Texture
    sortPriority=0
    texture=/Game/Weapons/FORT_Melee/Pickaxe_Future_Samurai_Summer_Male/FX/T_Pickaxe_Future_Samurai_Summer_Male_Trail_01
   }
  }
  MaterialExpressionTextureObjectParameter_1 {
   class=MaterialExpressionTextureObjectParameter
   inputs [
    UVs {from="", output=""}
    Apply View MipBias {from="", output=""}
   ]
   properties {
    group=BDBlendColor
    parameterName=BDBlendColor_Texture
    samplerType=SAMPLERTYPE_LinearGrayscale
    sortPriority=0
    texture=/Game/Effects/Fort_Effects/Emotes/UltraLight/Texture/T_Plasma_A_B_Greyscale
   }
  }
  MaterialExpressionVectorParameter_1 {
   class=MaterialExpressionVectorParameter
   properties {group=Edge, parameterName=Edge_UVControl, channelNames {a=VPan, b=UPan, g=VScale, r=UScale}, defaultValue=<rgba 1.4169 1.8 -2.5979 0>}
  }
  MaterialExpressionVectorParameter_2 {class=MaterialExpressionVectorParameter, properties {parameterName=Edge_Refine, defaultValue=<rgba 0.6657 14.7696 1 0.5>}}
  MaterialExpressionVectorParameter_3 {class=MaterialExpressionVectorParameter, properties {parameterName=EdgeC, defaultValue=<rgba 20 0.4 2.0333 1>}}
  MaterialExpressionVectorParameter_4 {class=MaterialExpressionVectorParameter, properties {parameterName=Edge_Offset, defaultValue=<rgba 0.17 0.0512 0 1>}}
  MaterialExpressionVectorParameter_5 {class=MaterialExpressionVectorParameter, properties {parameterName=edge2_Refine, defaultValue=<rgba 7 5.5468 0.3408 0.5>}}
  MaterialExpressionVectorParameter_6 {class=MaterialExpressionVectorParameter, properties {parameterName=EdgeC1, defaultValue=<rgba 0.5 0.05 0.0875 1>}}
  MaterialExpressionVectorParameter_7 {
   class=MaterialExpressionVectorParameter
   properties {group=BDBlendColor, parameterName=BDBlendColor_UVControl, channelNames {a=VPan, b=UPan, g=VScale, r=UScale}, defaultValue=<rgba 0.1 2 -2 0>}
  }
  MaterialExpressionVectorParameter_8 {class=MaterialExpressionVectorParameter, properties {parameterName=BDBlendColor_Refine, defaultValue=<rgba 2 24 13 0.5>}}
 }
 properties {
  bAutomaticallySetUsageInEditor=false
  bUsedWithNiagaraMeshParticles=true
  bUsedWithNiagaraRibbons=true
  bUsedWithNiagaraSprites=true
  blendMode=BLEND_Translucent
  shadingModel=MSM_Unlit
  usedShadingModels=Unlit
 }
}
