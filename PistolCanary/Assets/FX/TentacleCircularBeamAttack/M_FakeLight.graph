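# M_FakeLight is an additive, unlit Niagara material that fakes a light glow on scene geometry behind particles, filed with the tentacle circular beam attack FX.
# With depth testing off, a SphereMask compares world position behind translucency to particle position.
# Dynamic parameters shape the mask, though Radius also drives Hardness, before refine, particle color and emissive-adjustment steps.
# A separate chain multiplies platform-switched custom depth by a pixel-versus-scene depth cutoff, with no consumer shown.
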Material /PistolCanary/Assets/FX/TentacleCircularBeamAttack/M_FakeLight
graph {
 properties {
  bAutomaticallySetUsageInEditor=false
  bDisableDepthTest=true
  bUsedWithNiagaraMeshParticles=true
  bUsedWithNiagaraRibbons=true
  bUsedWithNiagaraSprites=true
  blendMode=BLEND_Additive
  shadingModel=MSM_Unlit
  twoSided=true
  usedShadingModels=Unlit
 }
 nodes [
  n1 = DynamicParameter() {defaultValue=<rgba 200 0 1 1>, paramNames [Radius,Hardness,Power,Param4]}
  n2 = Power(
   Base=SphereMask(
    A=MaterialFunctionCall() {materialFunction=&WorldPositionBehindTranslucency}.Result
    B=ParticlePositionWS()
    Radius=@n1.Radius
    Hardness=@n1.Radius
   )
   Exp=@n1.Power
  )
  n3 = VectorParameter() {parameterName=FakeLight_Refine, defaultValue=<rgba 0.9 3 1 0.5>}
  MaterialFunctionCall(Emissive=Multiply(
    A=MaterialFunctionCall(Texture=@n2, RGB=@n3.RGB, A=@n3.RGB) {materialFunction=&MF_FX_Refine}.Result
    B=ParticleColor().RGB
   )) {materialFunction=&MF_FX_EmissiveAdjustments}
  n4 = Saturate(None=Divide(A=SceneTexture() {sceneTextureId=PPI_CustomDepth}.Color) {constB=4096})
  n5 = Constant()
  n6 = PixelDepth()
  n7 = SceneDepth()
  n8 = Constant() {r=1}
  Multiply(
   A=MaterialFunctionCall("Xb1,PS4"=@n4, "PC-High&&Epic,XSX,PS5"=@n4, "PC-Perf,Android"=@n5, Switch=@n5, "PC-Low(NotPerf)&&Medium"=@n4) {materialFunction=&PlatformSwitch_1D}.Result
   B=If(A=Subtract(A=@n6.R, B=@n7.R), B=ScalarParameter() {defaultValue=500, parameterName=Cutoff}, "A > B"=Constant(), "A == B"=@n8, "A < B"=@n8)
  )
 ]
}
