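# Niagara system for the impact effect of a tentacle attack, filed under the TentacleCircularBeamAttack effects.
# It combines three emitters, Lightning, Smoke and Light, alongside system-level SystemState and two SetVariables modules.
# User parameters expose impact active, offset and scale floats plus bluish-purple primary and secondary colors.
# Its dynamic inputs include float curves, a waveform and color-from-vector conversion, though their wiring is not shown.
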NiagaraSystem /PistolCanary/Assets/FX/TentacleCircularBeamAttack/NS_PC_Tentacle_Attack_Impact
dependencies {
 usedDataInterfaces [&NiagaraDataInterfaceCurve]
 usedDynamicInputs [
  &Multiply_Float
  &FloatFromCurve
  &MakeLinearColorFromVectorAndFloat
  &VectorFromFloat
  &Waveform
 ]
 usedModules [
  &SystemState
  &self:SystemSpawnScript.SystemScriptSource.SystemScriptGraph.NiagaraNodeAssignment_0.SetVariables_#1
  &self:SystemSpawnScript.SystemScriptSource.SystemScriptGraph.NiagaraNodeAssignment_1.SetVariables_#2
 ]
}
emitters {Light {rendererClasses [], renderers []}, Lightning {rendererClasses [], renderers []}, Smoke {rendererClasses [], renderers []}}
summary {
 systemName=NS_PC_Tentacle_Attack_Impact
 emitters [{emitterName=Lightning, rendererClasses []}, {emitterName=Smoke, rendererClasses []}, {emitterName=Light, rendererClasses []}]
 userVariables [
  User.ImpactActive {type=NiagaraFloat, defaultValue=0, description=""}
  User.ImpactOffset {type=NiagaraFloat, defaultValue=0, description=""}
  User.ImpactScale {type=NiagaraFloat, defaultValue=2, description=""}
  User.PrimaryColor {type=LinearColor, description="", defaultValue {struct=LinearColor, value=<rgba 0.4969 0.8281 5 1>}}
  User.SecondaryColor {type=LinearColor, description="", defaultValue {struct=LinearColor, value=<rgba 0.3215 0.1494 0.7101 5>}}
 ]
}
