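# Niagara system for the end-point impact effect of PistolCanary's tentacle circular beam attack.
# Three emitters layer expanding mesh ground waves, flickering sprite flares at User.Stop Location and short-lived velocity-stretched spike sprites, all sized by BeamImpactScale and System.ExplosionSize.
# They share a delayed loop timed from User.Duration plus a system offset.
# Wave and spike spawn rates fall to zero along a curve indexed by System.TimeForDis, while flares spawn steadily.
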NiagaraSystem /PistolCanary/Assets/FX/TentacleCircularBeamAttack/NS_PistolCanary_Tentacle_Circular_Beam_End
dependencies {
 usedDataInterfaces [
  &NiagaraDataInterfaceCurve
  &NiagaraDataInterfaceVector2DCurve
  &NiagaraDataInterfaceColorCurve
  &NiagaraDataInterfaceCamera
 ]
 usedDynamicInputs [
  &MaskFloatBySpawnGroup
  &TransformPosition
  &MakeCustomFloatFromBool
  &Divide_Float
  &Subtract_Float
  &DistanceBetweenPositions
  &ReturnCameraPosition
 ]
 usedModules [
  &SystemState
  &Timeline
  &InterpolateOverTime
  &self:SystemSpawnScript.SystemScriptSource.SystemScriptGraph.NiagaraNodeAssignment_3.SetVariables_#1
  &self:SystemSpawnScript.SystemScriptSource.SystemScriptGraph.NiagaraNodeAssignment_1.SetVariables_#2
  &self:SystemSpawnScript.SystemScriptSource.SystemScriptGraph.NiagaraNodeAssignment_9.SetVariables_#3
  &DoOnce
  &self:SystemSpawnScript.SystemScriptSource.SystemScriptGraph.NiagaraNodeAssignment_12.SetVariables_#4
 ]
}
emitters {
 GroundWaveSlow001 {
  emitterUpdateScript [
   &EmitterState(
    Life Cycle Mode=::NewEnumerator1
    Loop Behavior=::NewEnumerator1
    Loop Delay=$System.TargetEXP_Delay
    Loop Duration=&Add_Float(A=$User.Duration, B=$System.TargetEXP_DurationOffset)
    UseLoopDelay=true
   )
   &SpawnRate(
    SpawnRate=&Multiply_Float(
     A=7
     B=&FloatFromCurve(
      FloatCurve=DataInterface() {propertyValues {Curve {keys [{interpMode=RCIM_Cubic, leaveTangent=-3.7201, tangentMode=RCTM_Break, time=0, value=1}, <at 1 0>]}}}
      CurveIndex=$System.TimeForDis
     )
    )
    Use Spawn Probability=false
   )
  ]
  particleSpawnScript [
   &InitializeParticle(
    Lifetime=0.25
    Mesh Scale=&Multiply_VectorByFloat(Vector=<1.8 1.8 1>, Float=$User.BeamImpactScale)
    Mesh Scale Mode=::NewEnumerator4
    UsePositionOffset=true
    Write Material Random=true
   )
   &InitialMeshOrientation(
    Mesh Axis To Orient=<0 0 -1>
    Mesh Orientation Mode=::NewEnumerator2
    Rotation=&RandomRangeVector(Minimum=<0 0 0>, Maximum=<0 0 1>)
    Use Rotation Vector=true
   )
   &InitialMeshOrientation(Mesh Orientation Mode=::NewEnumerator2, Rotation=<0 0.25 0>, Use Rotation Vector=true) {module=InitialMeshOrientation001}
   &CameraOffset()
  ]
  particleUpdateScript [
   &ParticleState()
   &ScaleMeshSize(Scale Factor=&Multiply_VectorByFloat(
     Vector=<1 1 1>
     Float=&FloatFromCurve(FloatCurve=DataInterface() {propertyValues {Curve {keys [<at 0 1>,<at 1 2.15>]}}})
    ))
   &ScaleColor(
    Scale Alpha=&FloatFromCurve(FloatCurve=DataInterface() {propertyValues {Curve {keys [
         {arriveTangent=26.1682, interpMode=RCIM_Cubic, leaveTangent=26.1682, tangentMode=RCTM_User, time=0, value=1}
         {arriveTangent=-1.4915, interpMode=RCIM_Cubic, leaveTangent=-1.4915, tangentMode=RCTM_User, time=0.1, value=1}
         {arriveTangent=-0.0587, tangentMode=RCTM_User, time=1, value=0}
        ]}}})
    Scale RGB=&MakeVectorFromLinearColorRGB(Color=&ColorFromCurve(ColorCurve=DataInterface() {propertyValues {
        AlphaCurve {keys []}
        BlueCurve {keys [<at 0.0679 2 RCIM_Cubic>, <at 0.2712 0.2>, <at 0.3632 0.2>, {interpMode=RCIM_Cubic, leaveTangent=0.0018, time=0.5109, value=1}]}
        GreenCurve {keys [<at 0.0679 2 RCIM_Cubic>, <at 0.2712 0.2>, <at 0.3632 0.2>, {interpMode=RCIM_Cubic, leaveTangent=0.1205, time=0.5109, value=1}]}
        RedCurve {keys [<at 0.0679 2 RCIM_Cubic>, <at 0.2712 0.2>, <at 0.3632 0.2>, {interpMode=RCIM_Cubic, leaveTangent=0.3657, time=0.5109, value=1}]}
       }}))
    ScaleA=true
    ScaleRGB=true
   )
   /SetVariables(Particles.Position=$User.Stop Location) {enabled=false, isSetParameters=true}
   &ScaleMeshSize(Scale Factor=&MakeVector(X=$System.ExplosionSize, Y=$System.ExplosionSize, Z=1)) {module=ScaleMeshSize001}
  ]
  rendererClasses [&NiagaraMeshRendererProperties]
  renderers [
   {rendererClass=&NiagaraMeshRendererProperties, rendererIndex=0}
   {rendererClass=&NiagaraMeshRendererProperties, rendererIndex=1}
  ]
 }
 Impact_Inner_Flares001 {
  emitterUpdateScript [
   &EmitterState(
    Life Cycle Mode=::NewEnumerator1
    Loop Behavior=::NewEnumerator1
    Loop Delay=$System.TargetEXP_Delay
    Loop Duration=&Add_Float(A=$User.Duration, B=$System.TargetEXP_DurationOffset)
    UseLoopDelay=true
   )
   &SpawnRate(SpawnRate=10, Use Spawn Probability=true)
  ]
  particleSpawnScript [
   &InitializeParticle(
    Color=<rgba 0.8362 2.7117 10 0.5>
    Lifetime Max=0.35
    Lifetime Min=0.15
    Lifetime Mode=::NewEnumerator1
    Position=$User.Stop Location
    Position Mode=::NewEnumerator1
    Sprite Rotation Mode=::NewEnumerator2
    Sprite Size Mode=::NewEnumerator1
    Uniform Sprite Size Max=&Multiply_Float(A=600, B=$User.BeamImpactScale)
    Uniform Sprite Size Min=&Multiply_Float(A=500, B=$User.BeamImpactScale)
    UsePositionOffset=false
    Write Material Random=true
   )
   &CameraOffset(Camera Offset Amount=200)
  ]
  particleUpdateScript [
   &ParticleState()
   &ScaleSpriteSize(Scale Sprite Size Mode=::NewEnumerator0, Uniform Scale Factor=$System.ExplosionSize)
   /SetVariables(Particles.Position=$User.Stop Location) {enabled=false, isSetParameters=true}
   &ScaleColor(
    Scale Alpha=&FloatFromCurve(FloatCurve=DataInterface() {propertyValues {Curve {keys [
         {interpMode=RCIM_Cubic, tangentMode=RCTM_User, time=0, value=0}
         {interpMode=RCIM_Cubic, tangentMode=RCTM_User, time=0.5, value=1}
         {tangentMode=RCTM_User, time=1, value=0}
        ]}}})
    Scale RGB=<2 2 2>
    ScaleA=true
    ScaleRGB=true
   )
   &DynamicMaterialParameters(
    Index 0 Param 1=100
    Index 0 Param 2=1
    Index 0 Param 3=1
    Index 0 Param 4=1
    Index 3 Param 1=&FloatFromCurve(FloatCurve=DataInterface() {propertyValues {Curve {keys [<at 0 0.7 RCIM_Cubic>,<at 0.5 0.25 RCIM_Cubic>,<at 1 0.7 RCIM_Cubic>]}}})
    Index 3 Param 2=1
    Index 3 Param 3=&RandomRangeFloat(Minimum=0.25, Maximum=0.35)
    Index 3 Param 4=1
    Module.Param0WriteEnabled=true
    Module.Param0WriteEnabled3=true
    Module.Param1WriteEnabled=true
    Module.Param1WriteEnabled3=true
    Module.Param2WriteEnabled=true
    Module.Param2WriteEnabled3=true
    Module.Param3WriteEnabled=true
    Module.Param3WriteEnabled3=true
    Write Parameter Index 3=true
   )
  ]
  rendererClasses [&NiagaraSpriteRendererProperties]
  renderers [{rendererClass=&NiagaraSpriteRendererProperties}]
 }
 spikes001 {
  emitterUpdateScript [
   &EmitterState(
    Life Cycle Mode=::NewEnumerator1
    Loop Behavior=::NewEnumerator1
    Loop Delay=$System.TargetEXP_Delay
    Loop Duration=&Add_Float(A=$User.Duration, B=$System.TargetEXP_DurationOffset)
    UseLoopDelay=true
   )
   &SpawnRate(
    SpawnRate=&Multiply_Float(
     A=75
     B=&FloatFromCurve(
      FloatCurve=DataInterface() {propertyValues {Curve {keys [{interpMode=RCIM_Cubic, leaveTangent=-3.7201, tangentMode=RCTM_Break, time=0, value=1}, <at 1 0>]}}}
      CurveIndex=$System.TimeForDis
     )
    )
    Use Spawn Probability=false
   )
  ]
  particleSpawnScript [
   &InitializeParticle(
    Color=<rgba 10.3883 0 50 1>
    Lifetime Max=0.15
    Lifetime Min=0.1
    Lifetime Mode=::NewEnumerator1
    Sprite Size Max=&Multiply_Vector2D_ByFloat(Float=$User.BeamImpactScale, Vector2D=<40 150>)
    Sprite Size Min=&Multiply_Vector2D_ByFloat(Float=$User.BeamImpactScale, Vector2D=<20 100>)
    Sprite Size Mode=::NewEnumerator2
    UsePositionOffset=true
    Write Material Random=true
   )
   &ShapeLocation(MaskSpawnGroup=false, Override Seed=false, Rotation Mode=::NewEnumerator1, Shape Origin=<0 0 0>, Sphere Radius=20, UseOwnerScale=true)
   &SubUVAnimation(End Frame=0, SubUV Animation Mode:ENiagara_SubUVLookupMode=::NewEnumerator2)
   &AddVelocity(
    Module.Use Falloff Distance=false
    Origin Offset=&AddVectorToPosition(Position=<0 0 -10>, Vector=<0 0 0>)
    Override Seed=false
    Velocity Mode=::NewEnumerator1
    Velocity Origin=<0 0 0>
    Velocity Speed=&RandomRangeFloat(Minimum=101.0873, Maximum=476.0403)
   )
  ]
  particleUpdateScript [
   &ParticleState()
   &Drag(Drag=6, Use Linear Drag=true, Use Rotational Drag=false)
   &SolveForcesAndVelocity(Clamp Velocity=false, Limit Acceleration=false)
   &ScaleSpriteSize(
    Non-Uniform Curve Scale=&Multiply_Vector2D_ByFloat(Float=$System.ExplosionSize, Vector2D=<1.1 1.3>)
    Non-Uniform Curve Sprite Scale=DataInterface() {propertyValues {
      XCurve {keys [{tangentMode=RCTM_User, time=0, value=1}, {tangentMode=RCTM_User, time=1, value=0.8}]}
      YCurve {keys [{tangentMode=RCTM_User, time=0, value=0.5}, {tangentMode=RCTM_User, time=1, value=1.5}]}
     }}
    Scale Sprite Size Mode=::NewEnumerator3
   )
   &ScaleColor(
    Scale Alpha=&FloatFromCurve(FloatCurve=DataInterface() {propertyValues {Curve {keys [<at 0 0>,<at 0.5 1>,<at 1 0>]}}})
    Scale RGB=&Multiply_VectorByFloat(Vector=<1 1 1>, Float=1)
    ScaleA=true
    ScaleRGB=true
   )
   /SetVariables(Particles.Color dark:LinearColor=&MakeLinearColorFromVectorAndFloat(
     "Vector (RGB)"=&MakeVectorFromLinearColorRGB(Color=&MultiplyLinearColor(Linear Color=<rgba 0.0198 0.0465 0.1 10>, Scale Factor=&MakeVector4_FromFloat(Float=1)))
     "Float (Alpha)"=&Multiply_Float(
      A=2
      B=&FloatFromCurve(FloatCurve=DataInterface() {propertyValues {Curve {keys [<at 0 0>,<at 0.5 1>,<at 1 0>]}}})
     )
    )) {enabled=false, isSetParameters=true}
   /SetVariables(Particles.Size02:Vector2f=&Multiply_Vector2D(
     A=&RandomRangeVector2D(Minimum=<20 120>, Maximum=<40 120>)
     B=&Vector2DFromCurve(
      Vector2Curve=DataInterface() {propertyValues {
        XCurve {keys [{tangentMode=RCTM_User, time=0, value=1}, {tangentMode=RCTM_User, time=1, value=0.8}]}
        YCurve {keys [{tangentMode=RCTM_User, time=0, value=0.5}, {tangentMode=RCTM_User, time=1, value=1.7}]}
       }}
      ScaleCurve=<2.5 2.5>
     )
    )) {enabled=false, isSetParameters=true}
  ]
  rendererClasses [&NiagaraSpriteRendererProperties]
  renderers [
   {rendererClass=&NiagaraSpriteRendererProperties}
   {rendererClass=&NiagaraSpriteRendererProperties, rendererIndex=1}
  ]
 }
}
summary {
 systemName=NS_PistolCanary_Tentacle_Circular_Beam_End
 emitters [
  {emitterName=spikes001, rendererClasses [&NiagaraSpriteRendererProperties]}
  {emitterName=Impact_Inner_Flares001, rendererClasses [&NiagaraSpriteRendererProperties]}
  {emitterName=GroundWaveSlow001, rendererClasses [&NiagaraMeshRendererProperties]}
 ]
 userVariables [
  User.BeamImpactScale {type=NiagaraFloat, defaultValue=5, description=""}
  User.bHitReaction {type=NiagaraBool}
  User.DisappearTime {type=NiagaraFloat, defaultValue=0.8, description=""}
  User.Duration {type=NiagaraFloat, defaultValue=10, description=""}
  User.InterpolateRateOfChange {type=NiagaraFloat, defaultValue=20, description=""}
  User.Stop Location {type=NiagaraPosition, description="", defaultValue {}}
 ]
}
