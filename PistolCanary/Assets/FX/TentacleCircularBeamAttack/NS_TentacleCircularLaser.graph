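# Niagara effect for the tentacle circular laser attack, layering energy bolts, charge swirl, flares and beam impact visuals.
# Bolt meshes and Charge-scaled rotating energy meshes play alongside a color-cycling GPU center flare with occlusion-driven visibility.
# Core flare, dark glow and ground waves sit at the system's stop location, shifted and scaled by User impact parameters.
# Several emitters, including Eyeglow, EnergyThorn_Edge, Front_LensFlare and PostProcess, are disabled.
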NiagaraSystem /PistolCanary/Assets/FX/TentacleCircularBeamAttack/NS_TentacleCircularLaser
dependencies {
 usedDataInterfaces [
  &NiagaraDataInterfaceCurve
  &NiagaraDataInterfaceArrayFloat3
  &NiagaraDataInterfaceArrayInt32
  &NiagaraDataInterfacePlatformSet
  &NiagaraDataInterfaceVectorCurve
  &NiagaraDataInterfaceColorCurve
  &NiagaraDataInterfaceVector2DCurve
  &NiagaraDataInterfaceCamera
  /Script/Niagara.NiagaraDataInterfaceOcclusion
  &NiagaraDataInterfaceSkeletalMesh
 ]
 usedDynamicInputs [&MaskFloatBySpawnGroup]
 usedModules [
  &SystemState
  &self:SystemSpawnScript.SystemScriptSource.SystemScriptGraph.NiagaraNodeAssignment_1.SetVariables_#1
 ]
}
emitters {
 Bolt {
  emitterUpdateScript [
   &EmitterState()
   &SpawnBurst_Instantaneous(Spawn Count=10, Use Spawn Probability=false, UseLoopCountLimit=false)
  ]
  particleSpawnScript [
   /SetVariables(Particles.int32:NiagaraInt32=&RandomRangeInt(Minimum=0, Maximum=1)) {isSetParameters=true}
   /SetVariables(Particles.MeshIndex=&RandomRangeInt(Minimum=0, Maximum=3)) {isSetParameters=true}
   &InitializeParticle(
    Color=<rgba 0 0.7151 5 1>
    Lifetime=&RandomRangeFloat(Minimum=0.55, Maximum=0.4)
    Mesh Scale=<0 0 0>
    Mesh Scale Mode=::NewEnumerator4
    Sprite Size Mode=::NewEnumerator3
    Uniform Sprite Size=800
    UsePositionOffset=false
    Write Material Random=true
   )
   &SystemLocation(MaskSpawnGroup=false, Offset=&RandomRangeVector(Minimum=<0 0 25>, Maximum=<0 0 25>))
   &InitialMeshOrientation(
    Mesh Orientation Mode=::NewEnumerator2
    Orientation Vector=&MaskVector(
     Mask Source=1
     Mask Parameter=$Particles.int32
     Mask=0
     Masked Vector=<-1 0 1>
     Passthrough Vector=&RandomRangeVector(Minimum=<-1 -1 -1>, Maximum=<1 1 1>)
    )
    Rotation=&MakeVector(X=&RandomRangeFloat(Minimum=0, Maximum=0.01), Y=-0.14, Z=0)
    Use Rotation Vector=true
   )
   &AddVelocityInCone(
    Cone Angle=90
    Cone Axis=<0 0 0.25>
    Module.Use Velocity Falloff On Cone Axis=false
    Velocity Distribution Along Cone Axis=0.25
    Velocity Strength=-25
   ) {enabled=false}
   &AddVelocity(Velocity=<0 0 1>) {enabled=false}
   /SetVariables(Particles.Scale=&Multiply_VectorByFloat(
     Vector=&MaskVector(Mask Source=0, Mask Parameter=?, Mask=$Particles.int32, Masked Vector=<0.05 0.05 0.07>, Passthrough Vector=<0.0333 0.0333 0.0467>)
     Float=&RemapRange(
      Input Value To Remap=&FN_LOD_Distance_Remap(Start Distance=5000, End Distance=15000, Minimum Output=0.65, Maximum Output=1)
      Input Min=0
      Input Max=1
      Output Min=0
      Output Max=1
     )
    )) {isSetParameters=true}
   &CameraOffset(Camera Offset Amount=&Lerp_Float(
     A=-200
     B=0
     Alpha=&FN_LOD_Distance_Remap(Start Distance=0, End Distance=1500, Minimum Output=0, Maximum Output=1)
    )) {enabled=false}
  ]
  particleUpdateScript [
   &SolveForcesAndVelocity(Clamp Velocity=false, Limit Acceleration=false)
   &ParticleState()
   &ScaleColor(
    Scale Alpha=&FloatFromCurve(FloatCurve=DataInterface() {propertyValues {Curve {keys [
         {interpMode=RCIM_Cubic, tangentMode=RCTM_User, time=0, value=0}
         {interpMode=RCIM_Cubic, tangentMode=RCTM_User, time=0.5, value=1}
         {tangentMode=RCTM_User, time=0.6, value=0}
        ]}}})
    Scale RGB=&VectorFromFloat(Value=&FloatFromCurve(
      FloatCurve=DataInterface() {propertyValues {Curve {keys [
          {interpMode=RCIM_Cubic, tangentMode=RCTM_User, time=0, value=0}
          {arriveTangent=-3.0443, interpMode=RCIM_Cubic, leaveTangent=-3.0443, tangentMode=RCTM_User, time=0.1, value=0.95}
          {interpMode=RCIM_Cubic, tangentMode=RCTM_User, time=0.4, value=0.55}
          {tangentMode=RCTM_User, time=1, value=0}
         ]}}}
      Scale Curve=15
     ))
    ScaleA=true
    ScaleRGB=true
   )
   &DynamicMaterialParameters(
    Index 0 Param 1=60
    Index 0 Param 2=&FloatFromCurve(FloatCurve=DataInterface() {propertyValues {Curve {keys [
         {interpMode=RCIM_Cubic, tangentMode=RCTM_User, time=0, value=0}
         {arriveTangent=-2.9636, interpMode=RCIM_Cubic, leaveTangent=-2.9636, tangentMode=RCTM_User, time=0.4, value=0.95}
         {tangentMode=RCTM_User, time=0.9, value=0}
        ]}}})
    Index 0 Param 3=&FloatFromCurve(
     FloatCurve=DataInterface() {propertyValues {Curve {keys [
         {interpMode=RCIM_Cubic, tangentMode=RCTM_User, time=0, value=0}
         {tangentMode=RCTM_User, time=1.5, value=0.85}
         {interpMode=RCIM_Cubic, tangentMode=RCTM_User, time=2.5, value=1}
        ]}}}
     Scale Curve=2
    )
    Index 0 Param 4=1
    Index 1 Param 1=&FloatFromCurve(FloatCurve=DataInterface() {propertyValues {Curve {keys [{interpMode=RCIM_Cubic, tangentMode=RCTM_User, time=0, value=1}, {interpMode=RCIM_Cubic, tangentMode=RCTM_User, time=1, value=0}]}}})
    Index 1 Param 2=50
    Index 1 Param 3=1
    Index 1 Param 4=1
    Index 2 Param 1=&FloatFromCurve(FloatCurve=DataInterface() {propertyValues {Curve {keys [{interpMode=RCIM_Cubic, tangentMode=RCTM_User, time=0, value=1}, {tangentMode=RCTM_User, time=1, value=0.5}]}}})
    Module.Param0WriteEnabled=true
    Module.Param0WriteEnabled1=true
    Module.Param0WriteEnabled2=true
    Module.Param1WriteEnabled=true
    Module.Param1WriteEnabled1=true
    Module.Param1WriteEnabled2=true
    Module.Param2WriteEnabled=true
    Module.Param2WriteEnabled1=true
    Module.Param2WriteEnabled2=true
    Module.Param3WriteEnabled=true
    Module.Param3WriteEnabled1=true
    Module.Param3WriteEnabled2=true
    Write Parameter Index 1=true
    Write Parameter Index 2=true
   )
   &ScaleMeshSize(Scale Factor=&MakeVector(
     X=&FloatFromCurve(
      FloatCurve=DataInterface() {propertyValues {Curve {keys [
          {arriveTangent=-5.4159, interpMode=RCIM_Cubic, leaveTangent=-5.4159, tangentMode=RCTM_User, time=0, value=1.5}
          {interpMode=RCIM_Cubic, tangentMode=RCTM_User, time=1, value=1.5}
         ]}}}
      Scale Curve=3
     )
     Y=&FloatFromCurve(FloatCurve=DataInterface() {propertyValues {Curve {keys [
          {arriveTangent=73.8235, interpMode=RCIM_Cubic, leaveTangent=73.8235, tangentMode=RCTM_User, time=0, value=0}
          {tangentMode=RCTM_User, time=0.05, value=1}
         ]}}})
     Z=&FloatFromCurve(
      FloatCurve=DataInterface() {propertyValues {Curve {keys [
          {arriveTangent=9.6743, interpMode=RCIM_Cubic, leaveTangent=9.6743, tangentMode=RCTM_User, time=0, value=0}
          {interpMode=RCIM_Cubic, tangentMode=RCTM_User, time=0.1588, value=1.0402}
          {interpMode=RCIM_Cubic, tangentMode=RCTM_User, time=0.9122, value=1.3662}
          <at 1 1>
         ]}}}
      Scale Curve=2
     )
    ))
  ]
  rendererClasses [&NiagaraMeshRendererProperties]
  renderers [{rendererClass=&NiagaraMeshRendererProperties, rendererIndex=0}]
 }
 CenterFlare {
  simTarget=GPUComputeSim
  emitterUpdateScript [
   &EmitterState(Life Cycle Mode=::NewEnumerator1, Loop Behavior=::NewEnumerator1, Loop Duration=20, UseLoopDelay=false)
   &SpawnBurst_Instantaneous(Use Spawn Probability=false, UseLoopCountLimit=false)
  ]
  particleSpawnScript [&InitializeParticle(
    Color=<rgba 1.5408 1.3217 2 1>
    Sprite Size=&Multiply_Vector2D_ByFloat(Float=50, Vector2D=<120 70>)
    Sprite Size Mode=::NewEnumerator4
    UsePositionOffset=false
    Write Material Random=true
   )]
  particleUpdateScript [
   &ParticleState(Kill Particles When Lifetime Has Elapsed=false, Let Infinitely Lived Particles Die When Emitter Deactivates=true)
   &self:MaskInScreenSpace_01() {module=MaskInScreenSpace_01}
   &self:ScratchModule_03(NumberOfRings=3, OcclusionSize=300, SamplesPerRing=8) {module=ScratchModule_03}
   &ScaleColor(
    Curve Index=$Transient.DistanceFromScreenEdge
    Linear Color Curve=DataInterface() {propertyValues {
      AlphaCurve {keys [<at 0 1>]}
      BlueCurve {keys [<at 0.0295 0.3803>,<at 0.8973 2>]}
      GreenCurve {keys [<at 0.0295 0>,<at 0.8973 1.6561>]}
      RedCurve {keys [<at 0.0295 0.5>,<at 0.8973 0.1108>]}
     }}
    Scale Mode=::NewEnumerator2
   ) {enabled=false, module=ScaleColor002}
   /SetVariables(Particles.Position=$Engine.Owner.Position) {enabled=false, isSetParameters=true}
   &ScaleColor(
    Scale Alpha=$Particles.Visibility
    Scale RGB=&MakeVectorFromLinearColorRGB(Color=&ColorFromCurve(
      ColorCurve=DataInterface() {propertyValues {
        AlphaCurve {keys [<at 0 1 RCIM_Cubic>]}
        BlueCurve {keys [<at 0 2.5 RCIM_Cubic>,<at 1 2.3743 RCIM_Cubic>]}
        GreenCurve {keys [<at 0 1.482 RCIM_Cubic>,<at 1 0 RCIM_Cubic>]}
        RedCurve {keys [<at 0 0 RCIM_Cubic>,<at 1 2.5 RCIM_Cubic>]}
       }}
      CurveIndex=&Waveform("Waveform [1]"=::NewEnumerator0, "[1] Frequency"=2, "[1] Amplitude Min/Max"=<0 1>)
     ))
    ScaleA=true
    ScaleRGB=true
   )
   &Color(Color=&ColorFromCurve(
     ColorCurve=DataInterface() {propertyValues {
       AlphaCurve {keys [<at 0 1 RCIM_Cubic>]}
       BlueCurve {keys [<at 0 2.5 RCIM_Cubic>,<at 1 2.3743 RCIM_Cubic>]}
       GreenCurve {keys [<at 0 1.482 RCIM_Cubic>,<at 1 0 RCIM_Cubic>]}
       RedCurve {keys [<at 0 0 RCIM_Cubic>,<at 1 2.5 RCIM_Cubic>]}
      }}
     CurveIndex=&Waveform("Waveform [1]"=::NewEnumerator0, "[1] Frequency"=2, "[1] Amplitude Min/Max"=<0 1>)
    )) {enabled=false}
   &ScaleColor(
    Color Value To Scale=$Particles.Color
    Scale Alpha=&Multiply_Float(A=$User.GlobalOpacity, B=1)
    Scale RGB=&Multiply_Vector(A=&MakeVectorFromLinearColorRGB(Color=$Particles.AtmosphereColor), B=<1 1 1>)
    ScaleA=true
    ScaleRGB=true
   ) {enabled=false, module=ScaleColor001}
  ]
  rendererClasses [&NiagaraSpriteRendererProperties]
  renderers [{rendererClass=&NiagaraSpriteRendererProperties}]
 }
 CoreFlare {
  emitterSpawnScript [/SetVariables(Emitter.SizeCurve=DataInterface() {propertyValues {Curve {keys [
        <at 0.2 0 RCIM_Cubic>
        {arriveTangent=2.9237, interpMode=RCIM_Cubic, leaveTangent=1.25, tangentMode=RCTM_Break, time=0.8, value=0.95}
        {arriveTangent=-0.193, interpMode=RCIM_Cubic, leaveTangent=-0.193, tangentMode=RCTM_User, time=2, value=0}
       ]}}}) {isSetParameters=true}]
  emitterUpdateScript [
   &EmitterState(Life Cycle Mode=::NewEnumerator1, Loop Behavior=::NewEnumerator1, Loop Duration=20, UseLoopDelay=true)
   &SpawnBurst_Instantaneous(Use Spawn Probability=false, UseLoopCountLimit=true)
  ]
  particleSpawnScript [&InitializeParticle(
    Color=<rgba 0 0.3661 1 3>
    Lifetime=14
    Position:Vector3f=$System.Stop Location Local
    Position Mode=::NewEnumerator1
    Sprite Size Mode=::NewEnumerator3
    Uniform Sprite Size=1200
    UsePositionOffset=true
    Write Material Random=true
   )]
  particleUpdateScript [
   &ParticleState()
   &ScaleSpriteSize(Scale Sprite Size Mode=::NewEnumerator0, Uniform Scale Factor=$User.ImpactScale)
   &CameraOffset(Camera Offset Amount=100)
   &ScaleColor(
    Scale Alpha=&Waveform(Global Amplitude Scale=0.2, "Waveform [1]"=::NewEnumerator12, "[1] Phase Scale"=10, "[1] Frequency"=100, "[1] Amplitude Min/Max"=<0.5 1>)
    ScaleA=true
    ScaleRGB=true
   )
   /SetVariables(Particles.Position=&AddVectorToPosition(Position=$Particles.Initial.Position, Vector=&MakeVector(X=$User.ImpactOffset, Y=0, Z=0))) {isSetParameters=true}
   &ScaleColor(
    Scale Alpha=&FloatFromCurve(
     FloatCurve=DataInterface() {propertyValues {Curve {keys [{interpMode=RCIM_Cubic, leaveTangent=-3.7201, tangentMode=RCTM_Break, time=0, value=1}, <at 1 0>]}}}
     CurveIndex=$System.TimeForDis
    )
    ScaleA=true
    ScaleRGB=true
   ) {module=ScaleColor002}
  ]
  rendererClasses [&NiagaraSpriteRendererProperties]
  renderers [{rendererClass=&NiagaraSpriteRendererProperties}]
 }
 EnergyThorn_Edge {
  bEnabled=false
  emitterUpdateScript [
   &EmitterState(Inactive Response=::NewEnumerator1, Life Cycle Mode=::NewEnumerator1, Loop Behavior=::NewEnumerator1, Loop Duration=5.5, UseLoopDelay=false)
   &SpawnRate(
    SpawnRate=&FloatFromCurve(
     FloatCurve=DataInterface() {propertyValues {Curve {
        defaultValue=1
        keys [
         {interpMode=RCIM_Cubic, tangentMode=RCTM_User, time=0, value=1}
         {interpMode=RCIM_Cubic, tangentMode=RCTM_User, time=0.9336, value=1}
         {tangentMode=RCTM_User, time=1, value=0}
        ]
       }}}
     CurveIndex=$Emitter.NormalizedLoopAge
     Scale Curve=20
    )
    Use Spawn Probability=false
   )
  ]
  particleSpawnScript [
   &PartitionParticles()
   &InitializeParticle(
    Color Maximum=<rgba 0.0469 0.0434 0.0398 1>
    Color Minimum=<rgba 0.1406 0.1301 0.1195 1.5>
    Color Mode=::NewEnumerator2
    Lifetime Max=0.3
    Lifetime Min=0.25
    Lifetime Mode=::NewEnumerator1
    Mesh Scale Max=<1.3 1.2 1>
    Mesh Scale Min=<1.3 0.8 0.8>
    Mesh Scale Mode=::NewEnumerator2
    UsePositionOffset=true
    Write Material Random=true
   )
   &AddVelocity(Velocity=&SelectVectorFromArray(
     Direct Array Index=$Particles.PartitionParticles.Partition
     Vector Selection Array=DataInterface() {propertyValues {FloatData [<100 15 20>, <-100 15 20>]}}
    ))
   &InitialMeshOrientation(Mesh Orientation Mode=::NewEnumerator2, Orientation Vector=<0 1 0>, Use Rotation Vector=true)
   &CameraOffset(Camera Offset Amount=100)
  ]
  particleUpdateScript [
   &ParticleState()
   &ScaleMeshSize(Scale Factor=<50 50 50>) {module=ScaleMeshSize001}
   &ScaleMeshSize(Scale Factor=&Multiply_Vector(
     A=&MakeVector(
      X=&FloatFromCurve(FloatCurve=DataInterface() {propertyValues {Curve {keys [<at 0 0.4>,<at 0.1396 0.8>,<at 1 1>]}}}, Scale Curve=1.2)
      Y=1
      Z=1
     )
     B=&Divide_Vector(A:Vector3f=<1 1 1>, B:Vector3f=$Engine.Owner.Scale)
    ))
   &ScaleColor(
    Scale Alpha=&Multiply_Float(
     A=&FloatFromCurve(
      FloatCurve=DataInterface() {propertyValues {Curve {keys [
          {arriveTangent=2.8247, interpMode=RCIM_Cubic, leaveTangent=2.8247, tangentMode=RCTM_User, time=0, value=0.5}
          {arriveTangent=0.0384, interpMode=RCIM_Cubic, leaveTangent=0.0384, tangentMode=RCTM_User, time=0.4, value=1}
          {arriveTangent=-3.1241, interpMode=RCIM_Cubic, leaveTangent=-3.1241, tangentMode=RCTM_User, time=1, value=0}
         ]}}}
      Scale Curve=&FNTimeOfDayScalar(Night Scale=0.3, Morning Scale=0.4, Day Scale=0.5, Evening Scale=0.3)
     )
     B=&FN_FrontEndFloatModifier(Gameplay=1, Front End=0.35)
    )
    Scale RGB=&MakeVectorFromLinearColorRGB(Color=&ColorFromCurve(ColorCurve=DataInterface() {propertyValues {
        AlphaCurve {keys []}
        BlueCurve {keys [
          {arriveTangent=-6.2021, interpMode=RCIM_Cubic, leaveTangent=-6.2021, tangentMode=RCTM_User, time=0.0978, value=15}
          {arriveTangent=-1.6366, interpMode=RCIM_Cubic, leaveTangent=-1.6366, tangentMode=RCTM_User, time=0.4613, value=5}
          {interpMode=RCIM_Cubic, leaveTangent=-0.0154, time=0.8372, value=50}
         ]}
        GreenCurve {keys [
          {arriveTangent=-7.4315, interpMode=RCIM_Cubic, leaveTangent=-7.4315, tangentMode=RCTM_User, time=0.0978, value=12.9851}
          {arriveTangent=-0.0423, interpMode=RCIM_Cubic, leaveTangent=-0.0423, tangentMode=RCTM_User, time=0.4613, value=1.0576}
          {interpMode=RCIM_Cubic, leaveTangent=-0.0154, time=0.8372, value=5.1307}
         ]}
        RedCurve {keys [
          {arriveTangent=-26.8321, interpMode=RCIM_Cubic, leaveTangent=-26.8321, tangentMode=RCTM_User, time=0.0978, value=2.522}
          {arriveTangent=19.5166, interpMode=RCIM_Cubic, leaveTangent=19.5166, time=0.4613, value=4.0437}
          {interpMode=RCIM_Cubic, leaveTangent=-0.0154, time=0.8372, value=16.9529}
         ]}
       }}))
    ScaleA=true
    ScaleRGB=true
   )
   &Drag(Drag=6, Use Linear Drag=true, Use Rotational Drag=false)
   &SolveForcesAndVelocity(Clamp Velocity=false, Limit Acceleration=false)
  ]
  rendererClasses [&NiagaraMeshRendererProperties]
  renderers [{rendererClass=&NiagaraMeshRendererProperties, rendererIndex=0}]
 }
 Energy_Offset {
  emitterUpdateScript [
   &EmitterState(Life Cycle Mode=::NewEnumerator1, Loop Behavior=::NewEnumerator1, Loop Delay=0.1, Loop Duration=14, UseLoopDelay=false)
   &SpawnBurst_Instantaneous(Spawn Count=2, Use Spawn Probability=false, UseLoopCountLimit=false)
   &SpawnRate(SpawnRate=6, Use Spawn Probability=false)
   /SetVariables(Emitter.Linear Color:LinearColor=&LerpLinearColors(
     StartColor=<rgba 0.0159 0.1401 1 1>
     EndColor=<rgba 0.7387 0.1312 1 1>
     LerpFactor=&Waveform("Waveform [1]"=::NewEnumerator0, "[1] Frequency"=2, "[1] Amplitude Min/Max"=<0 1>)
    )) {isSetParameters=true}
  ]
  particleSpawnScript [
   &InitializeParticle(Mesh Scale=<2 2 1>, Mesh Scale Mode=::NewEnumerator4, UsePositionOffset=true, Write Material Random=true)
   &InitialMeshOrientation(Mesh Axis To Orient=<0 0 1>, Mesh Orientation Mode=::NewEnumerator2, Use Rotation Vector=true)
   /SetVariables(Particles.MeshOrientation=&MultiplyQuaternion(
     Quaternion A=$Particles.InitialRotation
     Quaternion B=&MakeQuatFromAxisAngle(Axis=<0 0 1>, AngleInDegrees=&RandomRangeFloat(Minimum=-360, Maximum=360), NormalizeAxis=true)
    )) {isSetParameters=true}
  ]
  particleUpdateScript [
   &ParticleState()
   &ScaleMeshSize(Scale Factor=&VectorFromFloat(Value=&FloatFromCurve(FloatCurve=DataInterface() {propertyValues {Curve {keys [<at 0 0.5>,<at 0.1 1.2>]}}}, Scale Curve=$User.Charge)))
   &DynamicMaterialParameters(
    Index 0 Param 1=&FloatFromCurve(FloatCurve=DataInterface() {propertyValues {Curve {keys [<at 0 0.5>,<at 1 -0.63>]}}})
    Index 0 Param 2=&RandomRangeFloat(Minimum=-2, Maximum=1)
    Index 0 Param 3=0.1
    Module.Index0Param0Write=true
    Module.Index0Param1Write=true
    Module.Index0Param2Write=true
    Module.Index0Param3Write=false
   ) {module=DynamicMaterialParameters001}
   &ScaleColor(
    Scale Alpha=&FloatFromCurve(
     FloatCurve=DataInterface() {propertyValues {Curve {keys [<at 0 0>,<at 0.3 1>,<at 1 0>]}}}
     Scale Curve=&FloatFromCurve(FloatCurve=DataInterface() {propertyValues {Curve {keys [<at 0 0.5>,<at 0.2 1>]}}})
    )
    Scale RGB=&VectorFromFloat(Value=&FloatFromCurve(FloatCurve=DataInterface() {propertyValues {Curve {keys [<at 0 0>,<at 0.3 1>,<at 1 0>]}}}))
    ScaleA=true
    ScaleRGB=true
   )
   &UpdateMeshOrientation(Rotation Rate=0.025, Rotation Vector=<0 0 1>)
   &CameraOffset(Camera Offset Amount=350)
  ]
  rendererClasses [&NiagaraMeshRendererProperties]
  renderers [{rendererClass=&NiagaraMeshRendererProperties, rendererIndex=0}]
 }
 Eyeglow {
  bEnabled=false
  emitterUpdateScript [
   &EmitterState(Inactive Response=::NewEnumerator1, Life Cycle Mode=::NewEnumerator1, Loop Behavior=::NewEnumerator1, Loop Duration=5.5, UseLoopDelay=false)
   &SpawnBurst_Instantaneous(Use Spawn Probability=false, UseLoopCountLimit=false)
  ]
  particleSpawnScript [
   &InitializeParticle(
    Color=<rgba 0.15 0.15 0.15 1.5>
    Lifetime=5.5
    Mesh Scale=<0.05 0.08 0.05>
    Mesh Scale Mode=::NewEnumerator4
    Sprite Size=<3500 3500>
    Sprite Size Mode=::NewEnumerator4
    UsePositionOffset=false
    Write Material Random=true
   )
   &InitialMeshOrientation(Mesh Orientation Mode=::NewEnumerator2, Orientation Vector=<0 0 1>, Rotation=<0 0.25 0>, Use Rotation Vector=true)
   &CameraOffset(Camera Offset Amount=500)
   &AddVelocity(Velocity=<0 0 0.001>)
  ]
  particleUpdateScript [
   &ParticleState()
   &ScaleMeshSize(Scale Factor=&Multiply_Vector(
     A=&ScaleFloatByCurve(
      InitialValue:Vector3f=<1 1 1>
      ScaleCurve:NiagaraDataInterfaceVectorCurve=DataInterface() {propertyValues {
        XCurve {keys [<at 0 0.25>,<at 0.1 1>,<at 0.9 1>,<at 1 0>]}
        YCurve {keys [<at 0 1.25>,<at 0.1 2.25>,<at 0.9 2.25>,<at 1 0>]}
        ZCurve {keys [<at 0 1.25>,<at 0.1 2.25>,<at 0.9 2.25>,<at 1 0>]}
       }}
      CurveIndex:NiagaraFloat=$Particles.NormalizedAge
     )
     B=&Divide_Vector(A:Vector3f=<1 1 1>, B:Vector3f=$Engine.Owner.Scale)
    ))
   &ScaleSpriteSize(
    Uniform Curve Scale=3
    Uniform Curve Sprite Scale=DataInterface() {propertyValues {Curve {keys [
        <at 0 0>
        <at 0.1 1>
        <at 0.1512 0.7174>
        <at 0.2132 1>
        <at 0.2641 0.7539>
        <at 0.299 1>
        <at 0.3563 0.7037>
        <at 0.4151 1>
        <at 0.4564 0.8017>
        <at 0.5041 1>
        <at 0.5502 0.7561>
        <at 0.5836 1>
        <at 0.6551 0.7653>
        <at 0.6981 1>
        <at 0.7521 0.7721>
        <at 0.7871 1>
        <at 0.8475 0.7106>
        <at 0.9 1>
        {tangentMode=RCTM_User, time=1, value=0}
       ]}}}
   )
   /SetVariables(Particles.Size002:Vector2f=&Multiply_Vector2D(A=$Particles.SpriteSize, B=<0.6 0.9>)) {isSetParameters=true}
   /SetVariables(Particles.Size003:Vector2f=&Multiply_Vector2D(A=<150 50>, B=<1 3>)) {isSetParameters=true}
   &ScaleColor(
    Scale Alpha=&Multiply_Float(
     A=&FloatFromCurve(
      FloatCurve=DataInterface() {propertyValues {Curve {keys [
          {arriveTangent=4.4475, time=0, value=0}
          {arriveTangent=5.3518, leaveTangent=-3.1774, time=0.1, value=1.2}
          {arriveTangent=-1.4972, leaveTangent=-1.4972, time=0.1523, value=0.7}
          {arriveTangent=-0.1667, leaveTangent=-0.1667, time=0.2052, value=1.1711}
          <at 0.258 0.7713>
          <at 0.2926 1.1775>
          {arriveTangent=-0.25, leaveTangent=-0.25, time=0.3608, value=0.689}
          <at 0.3865 1.193>
          <at 0.4263 0.8642>
          <at 0.4931 1.1191>
          {arriveTangent=-0.25, leaveTangent=-0.25, time=0.5428, value=0.7548}
          <at 0.5846 1.1063>
          <at 0.6571 0.8076>
          <at 0.6976 1.1685>
          {arriveTangent=0.1667, leaveTangent=0.1667, time=0.7542, value=0.7713}
          <at 0.7906 1.1138>
          <at 0.8535 0.8807>
          {arriveTangent=-3.3333, leaveTangent=-3.3333, time=0.9, value=1.2}
          <at 1 0>
         ]}}}
      Scale Curve=&FNTimeOfDayScalar(Night Scale=0.6, Morning Scale=0.7, Day Scale=0.8, Evening Scale=0.6)
     )
     B=&FN_FrontEndFloatModifier(Gameplay=1, Front End=0.3)
    )
    Scale RGB=&MakeVectorFromLinearColorRGB(Color=&ColorFromCurve(ColorCurve=DataInterface() {propertyValues {
        AlphaCurve {keys []}
        BlueCurve {keys [<at 0.7887 0.4375 RCIM_Cubic>,<at 0.9801 0.1562 RCIM_Cubic>]}
        GreenCurve {keys [<at 0.7887 0.1271 RCIM_Cubic>,<at 0.9801 0.006 RCIM_Cubic>]}
        RedCurve {keys [<at 0.7887 0.0473 RCIM_Cubic>,<at 0.9801 0.0179 RCIM_Cubic>]}
       }}))
    ScaleA=true
    ScaleRGB=true
   )
   /SetVariables(Particles.color002:LinearColor=&MakeLinearColorFromVectorAndFloat(
     "Vector (RGB)"=&MakeVectorFromLinearColorRGB(Color=&ColorFromCurve(
       ColorCurve=DataInterface() {propertyValues {
         AlphaCurve {keys [<at 0 1 RCIM_Cubic>,<at 1 0 RCIM_Cubic>]}
         BlueCurve {keys [<at 0 1 RCIM_Cubic>,<at 1 1 RCIM_Cubic>]}
         GreenCurve {keys [<at 0 0.6237 RCIM_Cubic>,<at 1 0 RCIM_Cubic>]}
         RedCurve {keys [<at 0 0 RCIM_Cubic>,<at 1 0.5682 RCIM_Cubic>]}
        }}
       CurveIndex=&Waveform("Waveform [1]"=::NewEnumerator0, "[1] Frequency"=1, "[1] Amplitude Min/Max"=<0 1>)
      ))
     "Float (Alpha)"=&FloatFromCurve(
      FloatCurve=DataInterface() {propertyValues {Curve {keys [<at 0 1>,<at 0.1 2>,<at 0.2005 1.7767>,<at 0.306 1.9927>,<at 0.3889 1.7548>,<at 0.5026 2.0076>,<at 0.5773 1.7477>,<at 0.6601 1.9791>,<at 0.7283 1.7134>,<at 0.8209 1.9848>,<at 1 0>]}}}
      Scale Curve=&FN_PickFloatDependOnPlatform(TargetPlatformActive=5, TargetPlatformNoActive=1)
     )
    )) {isSetParameters=true}
   /SetVariables(Particles.Color003:LinearColor=&MakeLinearColorFromVectorAndFloat(
     "Vector (RGB)"=&MakeVectorFromLinearColorRGB(Color=<rgba 0.0114 0.1571 1 1>)
     "Float (Alpha)"=&Multiply_Float(
      A=&FloatFromCurve(
       FloatCurve=DataInterface() {propertyValues {Curve {keys [<at 0 0.8>,<at 0.1 1>,<at 0.2 0.6>,<at 0.3 1>,<at 0.4 0.7>,<at 0.5 0.95>,<at 0.6 0.55>,<at 0.7 1>,<at 0.8 0.55>,<at 0.9 0.8>,<at 1 0>]}}}
       Scale Curve=&FNTimeOfDayScalar(Night Scale=0.2, Morning Scale=0.1, Day Scale=0.4, Evening Scale=0.2)
      )
      B=&Multiply_Float(
       A=&FN_FrontEndFloatModifier(Gameplay=0.5, Front End=0.15)
       B=&FN_PickFloatDependOnPlatform(TargetPlatformActive=1, TargetPlatformNoActive=0.4)
      )
     )
    )) {isSetParameters=true}
   /SetVariables(Particles.MeshColor:LinearColor=&MakeLinearColorFromVectorAndFloat(
     "Vector (RGB)"=&MakeVectorFromLinearColorRGB(Color=&ColorFromCurve(ColorCurve=DataInterface() {propertyValues {
         AlphaCurve {keys []}
         BlueCurve {keys [
           {arriveTangent=-1.8762, interpMode=RCIM_Cubic, leaveTangent=-1.8762, tangentMode=RCTM_User, time=0.1298, value=2}
           {interpMode=RCIM_Cubic, leaveTangent=0.2132, time=0.3441, value=2}
          ]}
         GreenCurve {keys [
           {arriveTangent=-0.4331, interpMode=RCIM_Cubic, leaveTangent=-0.4331, tangentMode=RCTM_User, time=0.1298, value=1.336}
           {interpMode=RCIM_Cubic, leaveTangent=2.0123, time=0.3441, value=0.7834}
          ]}
         RedCurve {keys [
           {arriveTangent=-4.577, interpMode=RCIM_Cubic, leaveTangent=-4.577, tangentMode=RCTM_User, time=0.1298, value=0.0945}
           {interpMode=RCIM_Cubic, leaveTangent=-2.7231, time=0.3441, value=0.1072}
          ]}
        }}))
     "Float (Alpha)"=&Multiply_Float(
      A=&FloatFromCurve(
       FloatCurve=DataInterface() {propertyValues {Curve {keys [
           {arriveTangent=-0.9, time=0, value=0}
           {arriveTangent=-0.0998, leaveTangent=-0.0998, time=0.05, value=1}
           {arriveTangent=-1.0872, leaveTangent=-1.0872, time=0.1804, value=0.7734}
           <at 0.2985 0.937>
           <at 0.4531 0.6834>
           <at 0.5813 0.9005>
           <at 0.7511 0.6392>
           <at 0.8564 0.8>
           {arriveTangent=-1.5912, leaveTangent=-1.5912, time=1, value=0}
          ]}}}
       Scale Curve=&FN_FrontEndFloatModifier(Gameplay=1, Front End=0.5)
      )
      B=&FNTimeOfDayScalar(Night Scale=0.7, Morning Scale=0.5, Evening Scale=0.6)
     )
    )) {isSetParameters=true}
   &FN_Light_Attributes(
    Color=&ColorFromCurve(
     ColorCurve=DataInterface() {propertyValues {
       AlphaCurve {keys []}
       BlueCurve {keys [<at 0.0219 30>,<at 0.9359 10>]}
       GreenCurve {keys [<at 0.0219 24.1812>,<at 0.9359 0.0016>]}
       RedCurve {keys [<at 0.0219 0>,<at 0.9359 5.4213>]}
      }}
     CurveIndex=&Waveform("Waveform [1]"=::NewEnumerator0, "[1] Frequency"=3, "[1] Amplitude Min/Max"=<0 1>)
    )
    Create Color Attribute=true
    Create Exponent Attribute=true
    Create Position Attribute=true
    Create Radius Attribute=true
    Exponent=3
    Position=$Particles.SkeletalMeshLocation.SampledPosition
    Radius=&FloatFromCurve(
     FloatCurve=DataInterface() {propertyValues {Curve {keys [<at 0 0.4>,<at 0.1 1.5>,<at 0.2 1>,<at 0.5 1>,<at 0.8 0.8>,<at 1 0.4>]}}}
     Scale Curve=&Multiply_Float(A=&MakeCustomFloatFromBool(Bool=$User.bIsFrontEndPreview, True Float=40, False Float=80), B=1)
    )
    Scale Color=&FloatFromCurve(
     FloatCurve=DataInterface() {propertyValues {Curve {keys [<at 0 0.5>,<at 0.1 1.1>,<at 0.1401 0.4933>,<at 0.2 1.3742>,<at 0.2354 0.5667>,<at 0.3 1.1>,<at 0.3404 0.4667>,<at 0.4098 1.3991>,<at 0.4342 0.48>,<at 0.5 1.1>,<at 0.5311 0.5133>,<at 0.5829 1.3306>,<at 0.6376 0.5133>,<at 0.7 1.1>,<at 0.741 0.5067>,<at 0.7989 1.4926>,<at 0.8459 0.3133>,<at 0.9 1.1>,<at 1 0>]}}}
     Scale Curve=&Multiply_Float(A=5, B=&FNTimeOfDayScalar(Night Scale=0.3, Morning Scale=0.5, Day Scale=0.8, Evening Scale=0.4))
    )
   )
   &SolveForcesAndVelocity(Clamp Velocity=false, Limit Acceleration=false)
  ]
  rendererClasses [
   &NiagaraSpriteRendererProperties
   &NiagaraMeshRendererProperties
   &NiagaraLightRendererProperties
  ]
  renderers [
   {rendererClass=&NiagaraSpriteRendererProperties}
   {rendererClass=&NiagaraSpriteRendererProperties, rendererIndex=1}
   {rendererClass=&NiagaraSpriteRendererProperties, rendererIndex=2}
   {rendererClass=&NiagaraMeshRendererProperties, rendererIndex=3}
   {rendererClass=&NiagaraLightRendererProperties, rendererIndex=4}
  ]
 }
 Front_LensFlare {
  bEnabled=false
  simTarget=GPUComputeSim
  emitterUpdateScript [
   &EmitterState(Life Cycle Mode=::NewEnumerator1, Loop Duration=1, UseLoopDelay=false)
   &SpawnBurst_Instantaneous(
    Spawn Count=2
    Spawn Probability=&MakeCustomFloatFromBool(Bool=$User.VehicleHasDriver, True Float=1, False Float=0)
    Use Spawn Probability=true
    UseLoopCountLimit=false
   )
  ]
  particleSpawnScript [
   &PartitionParticles()
   &InitializeParticle(
    Color=<rgba 0.0594 0.5486 5 1>
    Position Offset=&SelectVectorFromArray(
     Direct Array Index=$Particles.PartitionParticles.Partition
     Vector Selection Array=DataInterface() {propertyValues {FloatData [<0 60 0>, <0 -62 0>]}}
    )
    Sprite Rotation Mode=::NewEnumerator0
    Sprite Size=<125 75>
    Sprite Size Mode=::NewEnumerator4
    UsePositionOffset=true
    Write Material Random=true
   )
   /SetVariables(
    Particles.Streak_01_Offset:NiagaraPosition=&AddVectorToPosition(Position=$Particles.Position, Vector=<0 0 0>)
    Particles.Streak_01_Size:Vector2f=&Multiply_Vector2D(A=<350 25>, B=&Vector2DFromFloat(Value=20))
   ) {isSetParameters=true}
  ]
  particleUpdateScript [
   &ParticleState(Kill Particles When Lifetime Has Elapsed=false, Let Infinitely Lived Particles Die When Emitter Deactivates=true)
   &ScaleSpriteSize(Uniform Curve Scale=&Multiply_Float(A=20, B=$User.BeamWidth)) {module=ScaleSpriteSize001}
   &ScaleSpriteSize(
    Uniform Curve Index=$Particles.CameraScale
    Uniform Curve Sprite Scale=DataInterface() {propertyValues {Curve {keys [
        {arriveTangent=4.7849, interpMode=RCIM_Cubic, leaveTangent=4.7849, tangentMode=RCTM_User, time=0, value=0.5}
        {arriveTangent=0.9156, interpMode=RCIM_Cubic, leaveTangent=0.9156, tangentMode=RCTM_User, time=0.3, value=1.275}
        {tangentMode=RCTM_User, time=1, value=1.5}
       ]}}}
   ) {enabled=false}
   &ScaleColor(Scale Alpha=$Particles.Visibility, Scale RGB=&VectorFromFloat(Value=3), ScaleA=true, ScaleRGB=true) {enabled=false, module=ScaleColor002}
   &ScaleColor(
    Scale Alpha=&FloatFromCurve(
     FloatCurve=DataInterface() {propertyValues {Curve {keys [
         {interpMode=RCIM_Cubic, tangentMode=RCTM_User, time=0, value=0}
         {arriveTangent=0.4636, interpMode=RCIM_Cubic, leaveTangent=0.4636, tangentMode=RCTM_User, time=0.2, value=0.8}
         {arriveTangentWeight=1.2316, interpMode=RCIM_Cubic, tangentMode=RCTM_User, time=1, value=1}
        ]}}}
     CurveIndex=$Particles.CameraScale
    )
    Scale RGB=&Multiply_VectorByFloat(Vector=<1 1 1>, Float=$Particles.CameraScale)
    ScaleA=true
    ScaleRGB=true
   ) {enabled=false}
   &Color(Color=&ColorFromCurve(
     ColorCurve=DataInterface() {propertyValues {
       AlphaCurve {keys [<at 0 1 RCIM_Cubic>]}
       BlueCurve {keys [<at 0 2.5 RCIM_Cubic>,<at 1 2.3743 RCIM_Cubic>]}
       GreenCurve {keys [<at 0 1.482 RCIM_Cubic>,<at 1 0 RCIM_Cubic>]}
       RedCurve {keys [<at 0 0 RCIM_Cubic>,<at 1 2.5 RCIM_Cubic>]}
      }}
     CurveIndex=&Waveform("Waveform [1]"=::NewEnumerator0, "[1] Frequency"=2, "[1] Amplitude Min/Max"=<0 1>)
    ))
   /SetVariables(Particles.Streak_01_Size:Vector2f=&Multiply_Vector2D_ByFloat(Float=$User.BeamWidth, Vector2D=$Particles.Initial.Streak_01_Size)) {isSetParameters=true}
  ]
  rendererClasses [&NiagaraSpriteRendererProperties]
  renderers [
   {rendererClass=&NiagaraSpriteRendererProperties}
   {rendererClass=&NiagaraSpriteRendererProperties, rendererIndex=1}
  ]
 }
 GroundWaveSlow {
  emitterUpdateScript [
   &EmitterState(Life Cycle Mode=::NewEnumerator1, Loop Behavior=::NewEnumerator1, Loop Duration=20, UseLoopDelay=true)
   &SpawnRate(
    SpawnRate=&Multiply_Float(
     A=7
     B=&FloatFromCurve(
      FloatCurve=DataInterface() {propertyValues {Curve {keys [{interpMode=RCIM_Cubic, leaveTangent=-3.7201, tangentMode=RCTM_Break, time=0, value=1}, <at 1 0>]}}}
      CurveIndex=$User.ImpactActive
     )
    )
    Use Spawn Probability=false
   )
  ]
  particleSpawnScript [
   &InitializeParticle(
    Lifetime=0.25
    Mesh Scale=<1.8 1.8 1.8>
    Mesh Scale Mode=::NewEnumerator4
    Position Offset=$System.Stop Location Local
    UsePositionOffset=true
    Write Material Random=true
   )
   &InitialMeshOrientation(
    Mesh Axis To Orient=<0 0 -1>
    Mesh Orientation Mode=::NewEnumerator2
    Orientation Coordinate Space=::World
    Orientation Vector=<0 0 1>
    Rotation=&RandomRangeVector(Minimum=<0 0 0>, Maximum=<0 0 1>)
    Use Rotation Vector=true
   )
   &InitialMeshOrientation(Mesh Orientation Mode=::NewEnumerator2, Rotation=<0 0.25 0>, Use Rotation Vector=true) {enabled=false, module=InitialMeshOrientation001}
   &CameraOffset()
  ]
  particleUpdateScript [
   &ParticleState()
   &ScaleMeshSize(Scale Factor=&Multiply_VectorByFloat(
     Vector=<1 1 1>
     Float=&FloatFromCurve(FloatCurve=DataInterface() {propertyValues {Curve {keys [<at 0 1>,<at 1 2.15>]}}})
    ))
   &ScaleColor(
    Scale Alpha=&FloatFromCurve(
     FloatCurve=DataInterface() {propertyValues {Curve {keys [
         {arriveTangent=26.1682, interpMode=RCIM_Cubic, leaveTangent=26.1682, tangentMode=RCTM_User, time=0, value=1}
         {arriveTangent=-1.4915, interpMode=RCIM_Cubic, leaveTangent=-1.4915, tangentMode=RCTM_User, time=0.1, value=1}
         {arriveTangent=-0.0587, tangentMode=RCTM_User, time=1, value=0}
        ]}}}
     Scale Curve=10
    )
    Scale RGB=&MakeVectorFromLinearColorRGB(Color=&ColorFromCurve(ColorCurve=DataInterface() {propertyValues {
        AlphaCurve {keys []}
        BlueCurve {keys [<at 0.0679 2 RCIM_Cubic>, <at 0.2712 0.2>, <at 0.3632 0.2>, {interpMode=RCIM_Cubic, leaveTangent=0.0018, time=0.5109, value=1}]}
        GreenCurve {keys [<at 0.0679 2 RCIM_Cubic>, <at 0.2712 0.2>, <at 0.3632 0.2>, {interpMode=RCIM_Cubic, leaveTangent=0.1205, time=0.5109, value=1}]}
        RedCurve {keys [<at 0.0679 2 RCIM_Cubic>, <at 0.2712 0.2>, <at 0.3632 0.2>, {interpMode=RCIM_Cubic, leaveTangent=0.3657, time=0.5109, value=1}]}
       }}))
    ScaleA=true
    ScaleRGB=true
   )
   /SetVariables(Particles.Position=&AddVectorToPosition(Position=$Particles.Initial.Position, Vector=&MakeVector(X=$User.ImpactOffset, Y=0, Z=0))) {isSetParameters=true}
   &ScaleMeshSize(Scale Factor=&MakeVector(X=$User.ImpactScale, Y=$User.ImpactScale, Z=1)) {module=ScaleMeshSize001}
  ]
  rendererClasses [&NiagaraMeshRendererProperties]
  renderers [{rendererClass=&NiagaraMeshRendererProperties, rendererIndex=0}]
 }
 NE_Juno_DarkGlow001 {
  emitterUpdateScript [
   &EmitterState(Life Cycle Mode=::NewEnumerator1, Loop Behavior=::NewEnumerator1, Loop Delay=$System.TargetEXP_Delay, Loop Duration=20, UseLoopDelay=true)
   &SpawnBurst_Instantaneous(Use Spawn Probability=false, UseLoopCountLimit=false)
  ]
  particleSpawnScript [
   &InitializeParticle(
    Color=<rgba 0.0007 0.0039 0.02 1>
    Lifetime=14
    Position:Vector3f=$System.Stop Location Local
    Position Mode=::NewEnumerator1
    Sprite Size Mode=::NewEnumerator3
    Uniform Sprite Size=700
    UsePositionOffset=false
    Write Material Random=true
   )
   &CameraOffset(Camera Offset Amount=150)
  ]
  particleUpdateScript [
   &ParticleState()
   &ScaleColor(
    Scale Alpha=&FloatFromCurve(FloatCurve=DataInterface() {propertyValues {Curve {keys [<at 0.1285 1 RCIM_Cubic>]}}})
    ScaleA=true
    ScaleRGB=true
   )
   /SetVariables(Particles.Position=&AddVectorToPosition(Position=$Particles.Initial.Position, Vector=&MakeVector(X=$User.ImpactOffset, Y=0, Z=0))) {isSetParameters=true}
   &ScaleColor(
    Scale Alpha=&FloatFromCurve(
     FloatCurve=DataInterface() {propertyValues {Curve {keys [{interpMode=RCIM_Cubic, leaveTangent=-3.7201, tangentMode=RCTM_Break, time=0, value=1}, <at 1 0>]}}}
     CurveIndex=$System.TimeForDis
    )
    ScaleA=true
    ScaleRGB=true
   ) {module=ScaleColor002}
   &ScaleSpriteSize(Scale Sprite Size Mode=::NewEnumerator0, Uniform Scale Factor=$User.ImpactScale)
  ]
  rendererClasses [&NiagaraSpriteRendererProperties]
  renderers [{rendererClass=&NiagaraSpriteRendererProperties}]
 }
 PostProcess {
  bEnabled=false
  emitterUpdateScript [
   &EmitterState(Life Cycle Mode=::NewEnumerator1, Loop Behavior=::NewEnumerator1, UseLoopDelay=false)
   &SpawnBurst_Instantaneous(Spawn Time=0.05, Use Spawn Probability=false, UseLoopCountLimit=false)
   &SpawnBurst_Instantaneous(Spawn Group=1, Use Spawn Probability=false, UseLoopCountLimit=false) {module=SpawnBurst_Instantaneous001}
  ]
  particleSpawnScript [
   &InitializeParticle(
    Mesh Scale Mode=::NewEnumerator3
    Sprite Size=&Multiply_Vector2D_ByFloat(Float=&Multiply_Float(A=$User.ElementSizeScale, B=$User.PostSizeScale), Vector2D=<3948.9736 3201.8015>)
    Sprite Size Mode=::NewEnumerator4
    UsePositionOffset=false
    Write Material Random=true
   )
   &CameraOffset()
   /SetVariables(Particles.VisibilityTag=&SpawnGroup()) {isSetParameters=true}
  ]
  particleUpdateScript [
   &ParticleState()
   &Timeline(
    Play Rate=&FloatFromCurve(
     FloatCurve=DataInterface() {propertyValues {Curve {keys [
         {interpMode=RCIM_Cubic, leaveTangent=-12.5334, tangentMode=RCTM_Break, time=0, value=4}
         {interpMode=RCIM_Cubic, tangentMode=RCTM_Break, time=0.7, value=1}
        ]}}}
     Scale Curve=0.692
    )
    Timeline End=1.0000000200408773e+20
   )
   &ScaleSpriteSize(Uniform Curve Sprite Scale=DataInterface() {propertyValues {Curve {keys [
        {arriveTangent=15.597, interpMode=RCIM_Cubic, leaveTangent=15.597, tangentMode=RCTM_User, time=0, value=0}
        {arriveTangent=0.3869, interpMode=RCIM_Cubic, leaveTangent=0.3869, tangentMode=RCTM_User, time=0.05, value=1}
        {arriveTangent=0.0224, interpMode=RCIM_Cubic, leaveTangent=0.0224, tangentMode=RCTM_User, time=1, value=1}
       ]}}})
   &DynamicMaterialParameters(
    Index 0 Param 1=&FloatFromCurve(
     FloatCurve=DataInterface() {propertyValues {Curve {keys [
         {interpMode=RCIM_Cubic, leaveTangent=-4.8343, tangentMode=RCTM_Break, time=0, value=2}
         {arriveTangent=-0.1577, interpMode=RCIM_Cubic, tangentMode=RCTM_Break, time=1, value=0}
        ]}}}
     Scale Curve=&Multiply_Float(A=2, B=$User.PostNoiseStrength)
    )
    Index 0 Param 2=&FloatFromCurve(
     FloatCurve=DataInterface() {propertyValues {Curve {keys [
         {interpMode=RCIM_Cubic, leaveTangent=-16.2213, tangentMode=RCTM_Break, time=0, value=4}
         {arriveTangent=-17.124, interpMode=RCIM_Cubic, tangentMode=RCTM_Break, time=0.25, value=0}
        ]}}}
     Scale Curve=&Multiply_Float(A=1, B=$User.PostStrength)
    )
    Index 0 Param 3=1
    Index 0 Param 4=1
    Index 1 Param 1=0.1
    Index 1 Param 2=1500
    Index 1 Param 3=0.5
    Index 1 Param 4=1
    Index 2 Param 1=$Particles.Timeline.Time
    Index 2 Param 2=$Particles.Timeline.Time
    Index 2 Param 3=$Particles.Timeline.Time
    Index 2 Param 4=&Multiply_Float(A=0.2, B=$User.ElementSizeScale)
    Index 3 Param 1=&FloatFromCurve(FloatCurve=DataInterface() {propertyValues {Curve {keys [<at 0.1 1>,<at 0.4 0>]}}})
    Index 3 Param 2=200
    Module.Index0Param0Write=true
    Module.Index0Param1Write=true
    Module.Index0Param2Write=true
    Module.Index0Param3Write=true
    Module.Index1Param0Write=true
    Module.Index1Param1Write=true
    Module.Index1Param2Write=true
    Module.Index1Param3Write=true
    Module.Index2Param0Write=true
    Module.Index2Param1Write=true
    Module.Index2Param2Write=true
    Module.Index2Param3Write=true
    Module.Index3Param0Write=true
    Module.Index3Param1Write=true
    Module.Index3Param2Write=true
    Module.Index3Param3Write=true
    Write Parameter Index 1=true
    Write Parameter Index 2=true
    Write Parameter Index 3=true
   )
   &ScaleColor(
    Scale Alpha=&FloatFromCurve(FloatCurve=DataInterface() {propertyValues {Curve {keys [{interpMode=RCIM_Cubic, leaveTangent=-6.3065, tangentMode=RCTM_Break, time=0.2, value=1}, <at 1 0>]}}})
    ScaleA=true
    ScaleRGB=true
   )
   /SetVariables(Particles.BlendTest:NiagaraFloat=&FloatFromCurve(
     FloatCurve=DataInterface() {propertyValues {Curve {keys [
         {arriveTangent=1.3123, interpMode=RCIM_Cubic, leaveTangent=1.3123, tangentMode=RCTM_User, time=0, value=0}
         {arriveTangent=0.0074, interpMode=RCIM_Cubic, leaveTangent=0.0074, tangentMode=RCTM_User, time=0.1, value=0.9847}
         {arriveTangent=-1.4231, interpMode=RCIM_Cubic, leaveTangent=-1.4231, tangentMode=RCTM_User, time=1, value=0}
        ]}}}
     Scale Curve=&Subtract_Float(
      A=1
      B=&FN_LOD_Distance_Remap(Start Distance=8000, End Distance=120000, Minimum Output=0, Maximum Output=1, Clamp Output=true)
     )
    )) {isSetParameters=true}
   /SetVariables(Particles.ComponentsEnabled:NiagaraBool=&Boolean_NotOperation(Bool=&SetBoolByFloatComparison(
      A=&FN_LOD_Distance_Remap(Start Distance=4000, End Distance=6000, Minimum Output=0, Maximum Output=1, Clamp Output=true)
      B=1
     ))) {isSetParameters=true}
   &KillParticles(Kill Particles=&SetBoolByFloatComparison(
     A=&FloatFromCurve(FloatCurve=DataInterface() {propertyValues {Curve {keys [<at 0.9 1>,<at 1 -0.01>]}}})
     B=0
     Comparison Type=::NewEnumerator6
    ))
  ]
  rendererClasses [&NiagaraSpriteRendererProperties, &NiagaraComponentRendererProperties]
  renderers [
   {rendererClass=&NiagaraSpriteRendererProperties}
   {rendererClass=&NiagaraComponentRendererProperties, rendererIndex=1}
  ]
 }
 Smoke_Sphere {
  bEnabled=false
  emitterUpdateScript [
   &EmitterState(Life Cycle Mode=::NewEnumerator1, Loop Delay=0.05, Loop Duration=0.1, UseLoopDelay=true)
   &SpawnRate(SpawnRate=50, Use Spawn Probability=false)
   &SpawnBurst_Instantaneous(Spawn Count=12, Use Spawn Probability=false)
  ]
  particleSpawnScript [
   &InitializeParticle(
    Color=<rgba 0.224 0.224 0.224 0.5>
    Lifetime=&RandomRangeFloat(Minimum=1, Maximum=2)
    Sprite Size Mode=::NewEnumerator1
    Uniform Sprite Size Max=800
    Uniform Sprite Size Min=400
    UsePositionOffset=false
    Write Material Random=true
   )
   /SetVariables(Particles.SpriteSize=&Multiply_Vector2D_ByFloat(
     Float=&FN_LOD_Distance_Remap(Start Distance=5000, End Distance=25000, Minimum Output=1, Maximum Output=3)
     Vector2D=&RandomRangeVector2D(
      Minimum=&Vector2DFromFloat(Value=600)
      Maximum=&Vector2DFromFloat(Value=1000)
     )
    )) {isSetParameters=true}
   /SetVariables(Particles.InitSpriteRotRate=&UniformRangedFloat(Minimum=-3, Maximum=3)) {isSetParameters=true}
   /SetVariables() {isSetParameters=true}
   &SphereLocation(
    MaskSpawnGroup=false
    Non Uniform Scale=<1 1 0.25>
    Override Seed=true
    Random Seed=20
    Sphere Coordinate Space=::World
    Sphere Origin:NiagaraPosition=&TransformVector(
     Vector=&AddVectorToPosition(Position=$Particles.Position, Vector=&MakeVector(X=$User.ImpactOffset, Y=0, Z=0))
     Destination Space=::Local
    )
    Sphere Radius=500
    Surface Only=false
    UseNonUniformScale=true
    UseOffset=false
    UseSphereOrientation=false
   )
   /SetVariables(Particles.Position=&MaskVectorBySpawnGroup(
     Masked Vector=&Add_Vector(A=Linked() {linkedVariable {name=Engine.Owner.Position, type=NiagaraPosition}}, B=<0 0 10>)
     Passthrough Vector=Linked() {linkedVariable {name=Particles.Position, type=NiagaraPosition}}
     Spawn Group Mask=1
    )) {enabled=false, isSetParameters=true}
   &AddVelocityFromPoint(
    Module.Use Falloff Distance=false
    Origin Offset=&MakeVector(X=$User.ImpactOffset, Y=0, Z=0)
    Velocity Strength=&RandomRangeFloat(Minimum=800, Maximum=50000)
   )
  ]
  particleUpdateScript [
   &ParticleState()
   &SpriteRotationRate(Rotation Rate=&Multiply_Float(
     A=$Particles.InitSpriteRotRate
     B=&FloatFromCurve(
      FloatCurve=DataInterface() {propertyValues {Curve {keys [
          {arriveTangent=-4.0107, interpMode=RCIM_Cubic, leaveTangent=-4.0107, tangentMode=RCTM_User, time=-0.0046, value=1.0113}
          {arriveTangent=-0.2496, interpMode=RCIM_Cubic, leaveTangent=-0.2496, tangentMode=RCTM_User, time=0.5, value=0.16}
         ]}}}
      CurveIndex=$Emitter.Age
      Scale Curve=150
     )
    ))
   &ScaleColor(
    Scale Alpha=&FloatFromCurve(FloatCurve=DataInterface() {propertyValues {Curve {keys [
         <at 0.1 0.75>
         {interpMode=RCIM_Cubic, tangentMode=RCTM_User, time=0.2, value=2.25}
         {interpMode=RCIM_Cubic, tangentMode=RCTM_User, time=0.7612, value=1.5}
         <at 1 0>
        ]}}})
    Scale RGB=&VectorFromFloat(Value=&FloatFromCurve(FloatCurve=DataInterface() {propertyValues {Curve {keys [{arriveTangent=1.8145, interpMode=RCIM_Cubic, leaveTangent=0.4403, tangentMode=RCTM_Break, time=0, value=0.4}, <at 0.4 0.625>, <at 1 1>]}}}))
    ScaleA=true
    ScaleRGB=true
   )
   /Niagara/Modules/Update/Forces/CurlNoiseForce(Noise Frequency=0.0012, Noise Strength=800, Pan Noise=true, Pan Noise Field=<0 0 1>)
   &Drag(Drag=20, Use Linear Drag=true, Use Rotational Drag=false)
   &SolveForcesAndVelocity(Clamp Velocity=false, Limit Acceleration=false)
   &ScaleSpriteSize(Scale Factor=&Vector2DFromFloat(Value=&FloatFromCurve(
      FloatCurve=DataInterface() {propertyValues {Curve {keys [
          {arriveTangent=6.2202, interpMode=RCIM_Cubic, leaveTangent=6.2202, tangentMode=RCTM_User, time=0, value=0.15}
          {arriveTangent=1.0025, interpMode=RCIM_Cubic, leaveTangent=1.0025, tangentMode=RCTM_User, time=0.2, value=0.8}
          {arriveTangent=0.0143, interpMode=RCIM_Cubic, leaveTangent=0.0143, tangentMode=RCTM_User, time=1.0463, value=1}
         ]}}}
      Scale Curve=3
     )))
   &DynamicMaterialParameters(
    Index 0 Param 1=&Add_Float(
     A=$Particles.RandomFloat
     B=&FloatFromCurve(
      FloatCurve=DataInterface() {propertyValues {Curve {keys [
          {arriveTangent=2.7533, interpMode=RCIM_Cubic, leaveTangent=1.8958, tangentMode=RCTM_User, time=0, value=1}
          {arriveTangent=0.8292, interpMode=RCIM_Cubic, leaveTangent=0.8292, tangentMode=RCTM_User, time=0.321, value=1.3453}
          {arriveTangent=0.0529, interpMode=RCIM_Cubic, leaveTangent=0.2021, tangentMode=RCTM_User, time=1, value=1.5}
         ]}}}
      CurveIndex=&Add_Float(A=&Multiply_Float(A=$Particles.Initial.MaterialRandom, B=0.5), B=$Particles.NormalizedAge)
     )
    )
    Index 0 Param 2=&FloatFromCurve(FloatCurve=DataInterface() {propertyValues {Curve {keys [
         {arriveTangent=0.0453, interpMode=RCIM_Cubic, tangentMode=RCTM_Break, time=0, value=-1}
         {arriveTangent=0.0604, interpMode=RCIM_Cubic, leaveTangent=0.0604, tangentMode=RCTM_User, time=1, value=-1}
        ]}}})
    Index 0 Param 3=&FloatFromCurve(FloatCurve=DataInterface() {propertyValues {Curve {keys [
         {arriveTangent=-53.9199, interpMode=RCIM_Cubic, leaveTangent=-8.9727, tangentMode=RCTM_User, time=0, value=1}
         {arriveTangent=0.2347, interpMode=RCIM_Cubic, leaveTangent=0.2723, tangentMode=RCTM_Break, time=0.25, value=0}
        ]}}})
    Index 0 Param 4=&FloatFromCurve(FloatCurve=DataInterface() {propertyValues {Curve {keys [<at 0.5 1>,<at 1 0>]}}})
    Index 1 Param 1=&FloatFromCurve(
     FloatCurve=DataInterface() {propertyValues {Curve {keys [
         {interpMode=RCIM_Cubic, leaveTangent=4.7088, tangentMode=RCTM_Break, time=0, value=0}
         {arriveTangent=4.2257, interpMode=RCIM_Cubic, leaveTangent=4.2257, tangentMode=RCTM_User, time=0.25, value=1}
        ]}}}
     Scale Curve=&RandomRangeFloat(Minimum=1.5, Maximum=1.5)
    )
    Index 1 Param 2=&FloatFromCurve(
     FloatCurve=DataInterface() {propertyValues {Curve {keys [<at 0 0.2>, {leaveTangent=0.9066, time=1, value=2}]}}}
     Scale Curve=&RandomRangeFloat(Minimum=0.5, Maximum=0.6)
    )
    Index 1 Param 3=&FloatFromCurve(FloatCurve=DataInterface() {propertyValues {Curve {keys [
         {interpMode=RCIM_Cubic, leaveTangent=6.3917, tangentMode=RCTM_Break, time=0, value=0.2}
         {arriveTangent=1.7549, interpMode=RCIM_Cubic, leaveTangent=10.6767, tangentMode=RCTM_User, time=0.25, value=1}
        ]}}})
    Index 1 Param 4=&FloatFromCurve(
     FloatCurve=DataInterface() {propertyValues {Curve {keys [{arriveTangent=-0.136, interpMode=RCIM_Cubic, leaveTangent=-0.136, tangentMode=RCTM_User, time=0, value=1}, <at 1 0.95>]}}}
     Scale Curve=&RandomRangeFloat(Minimum=1.1, Maximum=1.4)
    )
    Index 2 Param 1=&FloatFromCurve(FloatCurve=DataInterface() {propertyValues {Curve {keys [
         {interpMode=RCIM_Cubic, leaveTangent=4.2485, tangentMode=RCTM_Break, time=0, value=0}
         {arriveTangent=1.0548, interpMode=RCIM_Cubic, leaveTangent=1.0548, tangentMode=RCTM_User, time=0.17, value=0.5715}
         <at 1 1>
        ]}}})
    Index 2 Param 2=&FloatFromCurve(
     FloatCurve=DataInterface() {propertyValues {Curve {keys [
         {arriveTangent=1.9056, interpMode=RCIM_Cubic, leaveTangent=1.9056, tangentMode=RCTM_User, time=0, value=0.2}
         {arriveTangent=0.0259, interpMode=RCIM_Cubic, leaveTangent=0.0259, tangentMode=RCTM_User, time=1, value=1.4}
        ]}}}
     Scale Curve=&RandomRangeFloat(Minimum=0.2, Maximum=0.7)
    )
    Index 2 Param 3=&FloatFromCurve(FloatCurve=DataInterface() {propertyValues {Curve {keys [
         {interpMode=RCIM_Cubic, leaveTangent=2.3064, tangentMode=RCTM_Break, time=0, value=1}
         {interpMode=RCIM_Cubic, leaveTangent=11.1233, tangentMode=RCTM_Break, time=1, value=2}
        ]}}})
    Index 2 Param 4=&FloatFromCurve(
     FloatCurve=DataInterface() {propertyValues {Curve {keys [
         {arriveTangent=-0.6867, interpMode=RCIM_Cubic, leaveTangent=-0.6867, tangentMode=RCTM_User, time=0, value=1}
         {interpMode=RCIM_Cubic, leaveTangent=-1.2956, tangentMode=RCTM_Break, time=1, value=0.75}
        ]}}}
     Scale Curve=1.6
    )
    Module.Param0WriteEnabled=true
    Module.Param0WriteEnabled1=true
    Module.Param0WriteEnabled2=true
    Module.Param1WriteEnabled=false
    Module.Param1WriteEnabled1=true
    Module.Param1WriteEnabled2=true
    Module.Param2WriteEnabled=true
    Module.Param2WriteEnabled1=true
    Module.Param2WriteEnabled2=true
    Module.Param3WriteEnabled=true
    Module.Param3WriteEnabled1=true
    Module.Param3WriteEnabled2=true
    Write Parameter Index 1=true
    Write Parameter Index 2=true
   )
   &CameraOffset(Camera Offset Amount=35) {enabled=false}
  ]
  rendererClasses [&NiagaraSpriteRendererProperties]
  renderers [
   {rendererClass=&NiagaraSpriteRendererProperties}
   {rendererClass=&NiagaraSpriteRendererProperties, rendererIndex=1}
  ]
 }
 Sphere_ShockWave001 {
  emitterUpdateScript [
   &EmitterState(Inactive Response=::NewEnumerator1, Life Cycle Mode=::NewEnumerator1, Loop Behavior=::NewEnumerator1, Loop Duration=20, UseLoopDelay=true)
   &SpawnRate(
    SpawnRate=&Multiply_Float(
     A=5
     B=&FloatFromCurve(
      FloatCurve=DataInterface() {propertyValues {Curve {keys [{interpMode=RCIM_Cubic, leaveTangent=-3.7201, tangentMode=RCTM_Break, time=0, value=1}, <at 1 0>]}}}
      CurveIndex=$User.ImpactActive
     )
    )
    Use Spawn Probability=false
   )
   &SpawnBurst_Instantaneous(Spawn Time=0.2, Use Spawn Probability=false, UseLoopCountLimit=false) {enabled=false, module=SpawnBurst_Instantaneous001}
   &SpawnBurst_Instantaneous(Use Spawn Probability=false, UseLoopCountLimit=false)
  ]
  particleSpawnScript [
   &PartitionParticles()
   &InitializeParticle(
    Lifetime=0.3
    Mesh Scale=<3.6667 3.6667 3.6667>
    Mesh Scale Mode=::NewEnumerator4
    Position:Vector3f=$System.Stop Location Local
    Position Mode=::NewEnumerator1
    Position Offset=<30 0 0>
    UsePositionOffset=false
    Write Material Random=true
   )
   &InitialMeshOrientation(Mesh Orientation Mode=::NewEnumerator2, Rotation=<0 0.25 0>, Use Rotation Vector=true) {module=InitialMeshOrientation001}
   &CameraOffset()
  ]
  particleUpdateScript [
   &ParticleState()
   &ScaleMeshSize(Scale Factor=&Multiply_VectorByFloat(
     Vector=<1 1 1>
     Float=&FloatFromCurve(
      FloatCurve=DataInterface() {propertyValues {Curve {keys [
          {interpMode=RCIM_Cubic, tangentMode=RCTM_User, time=0, value=0}
          <at 0.503 0.6971>
          {interpMode=RCIM_Cubic, tangentMode=RCTM_User, time=1, value=1}
         ]}}}
      Scale Curve=0.8
     )
    ))
   &ScaleColor(
    Scale Alpha=&FloatFromCurve(
     FloatCurve=DataInterface() {propertyValues {Curve {keys [
         {tangentMode=RCTM_User, time=0, value=1}
         {arriveTangent=-0.0405, interpMode=RCIM_Cubic, leaveTangent=-0.0405, tangentMode=RCTM_User, time=0.5, value=0.5}
         {arriveTangent=-2.4747, interpMode=RCIM_Cubic, leaveTangent=-2.4747, tangentMode=RCTM_User, time=1, value=0}
        ]}}}
     Scale Curve=10
    )
    Scale RGB=&Multiply_VectorByFloat(
     Vector=&MakeVectorFromLinearColorRGB(Color=&ColorFromCurve(ColorCurve=DataInterface() {propertyValues {
         AlphaCurve {keys []}
         BlueCurve {keys [<at 0.3922 0.2 RCIM_Cubic>,<at 1 1 RCIM_Cubic>]}
         GreenCurve {keys [<at 0.3922 0.094 RCIM_Cubic>,<at 1 0 RCIM_Cubic>]}
         RedCurve {keys [<at 0.3922 0 RCIM_Cubic>,<at 1 0.2117 RCIM_Cubic>]}
        }}))
     Float=1
    )
    ScaleA=true
    ScaleRGB=true
   )
   &DynamicMaterialParameters(
    Index 0 Param 1=&FloatFromCurve(FloatCurve=DataInterface() {propertyValues {Curve {keys [<at 0 0.85>,<at 1 1>]}}})
    Index 0 Param 2=&FloatFromCurve(FloatCurve=DataInterface() {propertyValues {Curve {keys [
         {interpMode=RCIM_Cubic, tangentMode=RCTM_User, time=0, value=1}
         {arriveTangent=-1, interpMode=RCIM_Cubic, leaveTangent=-1, time=0.6, value=0.7}
         {interpMode=RCIM_Cubic, leaveTangent=-3.3753, time=1, value=0}
        ]}}})
    Index 0 Param 3=1
    Index 0 Param 4=1
    Module.Param0WriteEnabled=true
    Module.Param1WriteEnabled=true
    Module.Param2WriteEnabled=true
    Module.Param3WriteEnabled=true
   )
   &FN_Mesh_Attributes(
    Color=&MakeLinearColorFromVectorAndFloat(
     "Vector (RGB)"=&Multiply_Vector(
      A=&MakeVectorFromLinearColorRGB(Color=<rgba 0.5 0.2139 0.0666 2>)
      B=&MakeVectorFromLinearColorRGB(Color=&ColorFromCurve(ColorCurve=DataInterface() {propertyValues {
          AlphaCurve {keys [<at 0 1 RCIM_Cubic>]}
          BlueCurve {keys [<at 0 1 RCIM_Cubic>,<at 1 1 RCIM_Cubic>]}
          GreenCurve {keys [<at 0 1 RCIM_Cubic>,<at 1 0 RCIM_Cubic>]}
          RedCurve {keys [<at 0 1 RCIM_Cubic>,<at 1 0.2703 RCIM_Cubic>]}
         }}))
     )
     "Float (Alpha)"=&FloatFromCurve(
      FloatCurve=DataInterface() {propertyValues {Curve {keys [<at 0 0.5 RCIM_Cubic>, {arriveTangent=-1.2761, tangentMode=RCTM_User, time=0.7, value=0}]}}}
      Scale Curve=2.5
     )
    )
    Create Color Attribute=true
    Create Scale Attribute=true
    Scale=&Multiply_VectorByFloat(
     Vector=&Multiply_Vector(
      A=<1 1 1>
      B=&VectorFromFloat(Value=&FloatFromCurve(
        FloatCurve=DataInterface() {propertyValues {Curve {keys [{interpMode=RCIM_Cubic, tangentMode=RCTM_User, time=0, value=0}, {interpMode=RCIM_Cubic, tangentMode=RCTM_User, time=0.7, value=1}]}}}
        Scale Curve=&Multiply_Float(A=1, B=&Multiply_Float(A=0.3, B=$User.ImpactScale))
       ))
     )
     Float=2
    )
    Write to Module Namespace=true
   ) {module=FN_Mesh_Attributes001}
   /SetVariables(Particles.Position02:NiagaraPosition=&AddVectorToPosition(Position=$Particles.Initial.Position, Vector=&MakeVector(X=$User.ImpactOffset, Y=0, Z=0))) {isSetParameters=true}
   /SetVariables(Particles.Position=&AddVectorToPosition(Position=$Particles.Initial.Position, Vector=&MakeVector(X=$User.ImpactOffset, Y=0, Z=0))) {isSetParameters=true}
   &ScaleMeshSize(Scale Factor=&VectorFromFloat(Value=$User.ImpactScale)) {module=ScaleMeshSize001}
  ]
  rendererClasses [&NiagaraMeshRendererProperties]
  renderers [
   {rendererClass=&NiagaraMeshRendererProperties, rendererIndex=0}
   {rendererClass=&NiagaraMeshRendererProperties, rendererIndex=1}
  ]
 }
 Trail_Main {
  emitterUpdateScript [
   &EmitterState(Inactive Response=::NewEnumerator1, Life Cycle Mode=::NewEnumerator1, Loop Behavior=::NewEnumerator1, Loop Duration=20, UseLoopDelay=false)
   &SpawnBurst_Instantaneous(Spawn Count=2, Use Spawn Probability=false, UseLoopCountLimit=false)
   /SetVariables(Emitter.SplineWidth_Base:NiagaraFloat=&FloatFromCurve(
     FloatCurve=DataInterface() {propertyValues {Curve {keys [
         <at 0 6>
         <at 0.1 1>
         {arriveTangent=-6.7804, interpMode=RCIM_Cubic, leaveTangent=-6.7804, tangentMode=RCTM_User, time=0.7, value=1}
         {arriveTangent=-0.8494, interpMode=RCIM_Cubic, leaveTangent=-0.8494, tangentMode=RCTM_User, time=1, value=0}
        ]}}}
     CurveIndex=$Emitter.NormalizedLoopAge
     Scale Curve=10
    )) {isSetParameters=true}
   /SetVariables(Emitter.SplineWidth_Tip:NiagaraFloat=&FloatFromCurve(
     FloatCurve=DataInterface() {propertyValues {Curve {keys [
         <at 0 1>
         {arriveTangent=-6.7804, interpMode=RCIM_Cubic, leaveTangent=-6.7804, tangentMode=RCTM_User, time=0.7, value=1}
         {arriveTangent=-0.8494, interpMode=RCIM_Cubic, leaveTangent=-0.8494, tangentMode=RCTM_User, time=1, value=0}
        ]}}}
     CurveIndex=$Emitter.NormalizedLoopAge
     Scale Curve=100
    )) {isSetParameters=true}
   /SetVariables(Emitter.SplineWidthRed_Base:NiagaraFloat=&FloatFromCurve(
     FloatCurve=DataInterface() {propertyValues {Curve {keys [
         <at 0 6>
         <at 0.1 1>
         {arriveTangent=-4.6776, interpMode=RCIM_Cubic, leaveTangent=-4.6776, tangentMode=RCTM_User, time=0.6, value=1}
         {arriveTangent=-0.5343, interpMode=RCIM_Cubic, leaveTangent=-0.5343, tangentMode=RCTM_User, time=1, value=0}
        ]}}}
     CurveIndex=$Emitter.NormalizedLoopAge
     Scale Curve=300
    )) {enabled=false, isSetParameters=true}
   /SetVariables(Emitter.SplineWidthRed_Tip:NiagaraFloat=&FloatFromCurve(
     FloatCurve=DataInterface() {propertyValues {Curve {keys [
         <at 0 1>
         {arriveTangent=-4.6776, interpMode=RCIM_Cubic, leaveTangent=-4.6776, tangentMode=RCTM_User, time=0.6, value=1}
         {arriveTangent=-0.5343, interpMode=RCIM_Cubic, leaveTangent=-0.5343, tangentMode=RCTM_User, time=1, value=0}
        ]}}}
     CurveIndex=$Emitter.NormalizedLoopAge
     Scale Curve=300
    )) {enabled=false, isSetParameters=true}
  ]
  particleSpawnScript [
   &PartitionParticles()
   &InitializeParticle(
    Color=<rgba 0.0706 0 0.15 1.5>
    Lifetime=20
    Mesh Scale=<800 700 400>
    Mesh Scale Mode=::NewEnumerator4
    UsePositionOffset=false
    Write Material Random=true
   )
   &InitialMeshOrientation(
    Mesh Axis To Orient=<0 0 1>
    Mesh Orientation Mode=::NewEnumerator2
    Orientation Coordinate Space=::Local
    Orientation Vector=<0 0 0>
    Rotation=&SelectVectorFromArray(
     Direct Array Index=$Particles.PartitionParticles.Partition
     Vector Selection Array=DataInterface() {propertyValues {FloatData [<0 0 0>, <0 0 0.25>]}}
    )
    Use Rotation Vector=true
   )
   /SetVariables(Particles.VisibilityTag=&SelectIntFromArray(
     Direct Array Index=&MakeCustomIntFromBool(
      Bool=&SetBoolByFloatComparison(A=0.5, B=&RandomRangeFloat(Minimum=0, Maximum=1))
      True INT=&MakeCustomIntFromBool(
       Bool=&SetBoolByFloatComparison(A=0.5, B=&RandomRangeFloat(Minimum=0, Maximum=1))
       True INT=1
       False INT=2
      )
      False INT=0
     )
     Int Selection Array=DataInterface() {propertyValues {IntData [0,1,2]}}
    )) {enabled=false, isSetParameters=true}
  ]
  particleUpdateScript [
   &ParticleState()
   &ScaleMeshSize(Scale Factor=&Multiply_Vector(
     A=&MakeVector(
      X=$User.BeamWidth
      Y=1
      Z=&FloatFromCurve(FloatCurve=DataInterface() {propertyValues {Curve {keys [<at 0 0>,<at 0.0063 1>,<at 1 1>]}}})
     )
     B=<1 1 1>
    ))
   &ScaleColor(
    Scale Alpha=&Multiply_Float(
     A=&FloatFromCurve(
      FloatCurve=DataInterface() {propertyValues {Curve {keys [
          <at 0 1>
          {arriveTangent=-0.2775, interpMode=RCIM_Cubic, leaveTangent=-0.2775, tangentMode=RCTM_User, time=0.8, value=1}
          {arriveTangent=-2.6012, interpMode=RCIM_Cubic, leaveTangent=-2.6012, tangentMode=RCTM_User, time=0.9, value=0.85}
          {arriveTangent=-14.0132, interpMode=RCIM_Cubic, leaveTangent=-14.0132, tangentMode=RCTM_User, time=1, value=0}
         ]}}}
      Scale Curve=&FNTimeOfDayScalar(Night Scale=0.6, Morning Scale=0.8, Evening Scale=0.6)
     )
     B=&FN_FrontEndFloatModifier(Gameplay=1, Front End=0.7)
    )
    ScaleA=true
    ScaleRGB=true
   )
   /SetVariables(Particles.color002:LinearColor=&MakeLinearColorFromVectorAndFloat(
     "Vector (RGB)"=&MakeVectorFromLinearColorRGB(Color=<rgba 1 1 1 1>)
     "Float (Alpha)"=&Multiply_Float(
      A=&FloatFromCurve(
       FloatCurve=DataInterface() {propertyValues {Curve {keys [<at 0 1>,<at 0.95 1>,<at 1 0>]}}}
       Scale Curve=&FNTimeOfDayScalar(Night Scale=0.2, Morning Scale=0.2, Day Scale=0.3, Evening Scale=0.2)
      )
      B=&Multiply_Float(
       A=&FN_FrontEndFloatModifier(Gameplay=1, Front End=0.8)
       B=&FN_PickFloatDependOnPlatform(TargetPlatformActive=1, TargetPlatformNoActive=0.8)
      )
     )
    )) {isSetParameters=true}
   /SetVariables(Particles.Color003:LinearColor=&MakeLinearColorFromVectorAndFloat(
     "Vector (RGB)"=&MakeVectorFromLinearColorRGB(Color=<rgba 1.5 1.5 1.5 1>)
     "Float (Alpha)"=&Multiply_Float(
      A=&FloatFromCurve(
       FloatCurve=DataInterface() {propertyValues {Curve {keys [<at 0 5>,<at 0.1 1>,<at 0.9 1>,<at 1 0>]}}}
       Scale Curve=&FNTimeOfDayScalar(Night Scale=0.6, Morning Scale=0.8, Evening Scale=0.6)
      )
      B=&FN_PickFloatDependOnPlatform(TargetPlatformActive=1, TargetPlatformNoActive=0.15)
     )
    )) {isSetParameters=true}
   /SetVariables(Particles.Color004:LinearColor=&MakeLinearColorFromVectorAndFloat(
     "Vector (RGB)"=&MakeVectorFromLinearColorRGB(Color=<rgba 1 1 1 1.5>)
     "Float (Alpha)"=&Multiply_Float(
      A=&FloatFromCurve(
       FloatCurve=DataInterface() {propertyValues {Curve {keys [<at 0 1>,<at 0.85 1>,<at 1 0>]}}}
       Scale Curve=&FNTimeOfDayScalar(Night Scale=0.6, Morning Scale=0.8, Evening Scale=0.6)
      )
      B=&FN_PickFloatDependOnPlatform(TargetPlatformActive=1, TargetPlatformNoActive=0.6)
     )
    )) {isSetParameters=true}
   &DynamicMaterialParameters(
    Index 0 Param 1=&FloatFromCurve(
     FloatCurve=DataInterface() {propertyValues {Curve {keys [
         {arriveTangent=12.0832, interpMode=RCIM_Cubic, leaveTangent=12.0832, tangentMode=RCTM_User, time=0.7, value=-1}
         {arriveTangent=0.7491, interpMode=RCIM_Cubic, leaveTangent=0.7491, tangentMode=RCTM_User, time=1, value=1}
        ]}}}
     Scale Curve=&RandomRangeFloat(Minimum=0.9, Maximum=1)
    )
    Index 0 Param 2=1
    Index 0 Param 3=1
    Index 0 Param 4=1
    Module.Index0Param0Write=true
    Module.Index0Param1Write=true
    Module.Index0Param2Write=true
    Module.Index0Param3Write=true
   )
   &CameraOffset(Camera Offset Amount=100)
  ]
  rendererClasses [&NiagaraMeshRendererProperties]
  renderers [
   {rendererClass=&NiagaraMeshRendererProperties, rendererIndex=0}
   {rendererClass=&NiagaraMeshRendererProperties, rendererIndex=1}
   {rendererClass=&NiagaraMeshRendererProperties, rendererIndex=2}
   {rendererClass=&NiagaraMeshRendererProperties, rendererIndex=3}
  ]
 }
 spikes {
  emitterUpdateScript [
   &EmitterState(Life Cycle Mode=::NewEnumerator1, Loop Behavior=::NewEnumerator1, Loop Duration=20, UseLoopDelay=true)
   &SpawnRate(
    SpawnRate=&Multiply_Float(
     A=75
     B=&FloatFromCurve(
      FloatCurve=DataInterface() {propertyValues {Curve {keys [{interpMode=RCIM_Cubic, leaveTangent=-3.7201, tangentMode=RCTM_Break, time=0, value=1}, <at 1 0>]}}}
      CurveIndex=$User.ImpactActive
     )
    )
    Use Spawn Probability=false
   )
  ]
  particleSpawnScript [
   &InitializeParticle(
    Color=<rgba 10.3883 0 50 1>
    Lifetime Max=0.15
    Lifetime Min=0.1
    Lifetime Mode=::NewEnumerator1
    Sprite Size Max=<50 127.974>
    Sprite Size Min=<29.1946 61.642>
    Sprite Size Mode=::NewEnumerator2
    UsePositionOffset=true
    Write Material Random=true
   )
   &ShapeLocation(MaskSpawnGroup=false, Override Seed=false, Rotation Mode=::NewEnumerator1, Shape Origin=<0 0 0>, Sphere Radius=20, UseOwnerScale=true)
   &SubUVAnimation(End Frame=0, SubUV Animation Mode:ENiagara_SubUVLookupMode=::NewEnumerator2)
   &AddVelocity(
    Module.Use Falloff Distance=false
    Origin Offset=&AddVectorToPosition(Position=<0 0 -10>, Vector=<0 0 0>)
    Override Seed=false
    Velocity Mode=::NewEnumerator1
    Velocity Origin=<0 0 0>
    Velocity Speed=&RandomRangeFloat(Minimum=101.0873, Maximum=476.0403)
   )
  ]
  particleUpdateScript [
   &ParticleState()
   &Drag(Drag=6, Use Linear Drag=true, Use Rotational Drag=false)
   &SolveForcesAndVelocity(Clamp Velocity=false, Limit Acceleration=false)
   &ScaleSpriteSize(
    Non-Uniform Curve Scale=&Multiply_Vector2D_ByFloat(Float=$User.ImpactScale, Vector2D=<1.1 1.3>)
    Non-Uniform Curve Sprite Scale=DataInterface() {propertyValues {
      XCurve {keys [{tangentMode=RCTM_User, time=0, value=1}, {tangentMode=RCTM_User, time=1, value=0.8}]}
      YCurve {keys [{tangentMode=RCTM_User, time=0, value=0.5}, {tangentMode=RCTM_User, time=1, value=1.5}]}
     }}
    Scale Sprite Size Mode=::NewEnumerator3
   )
   &ScaleColor(
    Scale Alpha=&FloatFromCurve(FloatCurve=DataInterface() {propertyValues {Curve {keys [<at 0 0>,<at 0.5 1>,<at 1 0>]}}})
    Scale RGB=&Multiply_VectorByFloat(Vector=<1 1 1>, Float=1)
    ScaleA=true
    ScaleRGB=true
   )
   /SetVariables(Particles.Color dark:LinearColor=&MakeLinearColorFromVectorAndFloat(
     "Vector (RGB)"=&MakeVectorFromLinearColorRGB(Color=&MultiplyLinearColor(Linear Color=<rgba 0.0198 0.0465 0.1 10>, Scale Factor=&MakeVector4_FromFloat(Float=1)))
     "Float (Alpha)"=&Multiply_Float(
      A=2
      B=&FloatFromCurve(FloatCurve=DataInterface() {propertyValues {Curve {keys [<at 0 0>,<at 0.5 1>,<at 1 0>]}}})
     )
    )) {enabled=false, isSetParameters=true}
   /SetVariables(Particles.Size02:Vector2f=&Multiply_Vector2D(
     A=&RandomRangeVector2D(Minimum=<20 120>, Maximum=<40 120>)
     B=&Vector2DFromCurve(
      Vector2Curve=DataInterface() {propertyValues {
        XCurve {keys [{tangentMode=RCTM_User, time=0, value=1}, {tangentMode=RCTM_User, time=1, value=0.8}]}
        YCurve {keys [{tangentMode=RCTM_User, time=0, value=0.5}, {tangentMode=RCTM_User, time=1, value=1.7}]}
       }}
      ScaleCurve=<2.5 2.5>
     )
    )) {enabled=false, isSetParameters=true}
   /SetVariables(Particles.Position2:NiagaraPosition=&AddVectorToPosition(Position=$Particles.Initial.Position, Vector=&MakeVector(X=$User.ImpactOffset, Y=0, Z=0))) {isSetParameters=true}
  ]
  rendererClasses [&NiagaraSpriteRendererProperties]
  renderers [
   {rendererClass=&NiagaraSpriteRendererProperties}
   {rendererClass=&NiagaraSpriteRendererProperties, rendererIndex=1}
  ]
 }
}
summary {
 systemName=NS_TentacleCircularLaser
 emitters [
  {emitterName=Trail_Main, rendererClasses [&NiagaraMeshRendererProperties]}
  {
   bEnabled=false
   emitterName=Eyeglow
   rendererClasses [
    &NiagaraSpriteRendererProperties
    &NiagaraMeshRendererProperties
    &NiagaraLightRendererProperties
   ]
  }
  {bEnabled=false, emitterName=EnergyThorn_Edge, rendererClasses [&NiagaraMeshRendererProperties]}
  {emitterName=Bolt, rendererClasses [&NiagaraMeshRendererProperties]}
  {
   bEnabled=false
   emitterName=PostProcess
   rendererClasses [&NiagaraSpriteRendererProperties, &NiagaraComponentRendererProperties]
  }
  {bEnabled=false, emitterName=Front_LensFlare, simTarget=GPUComputeSim, rendererClasses [&NiagaraSpriteRendererProperties]}
  {emitterName=Energy_Offset, rendererClasses [&NiagaraMeshRendererProperties]}
  {emitterName=GroundWaveSlow, rendererClasses [&NiagaraMeshRendererProperties]}
  {emitterName=Sphere_ShockWave001, rendererClasses [&NiagaraMeshRendererProperties]}
  {emitterName=spikes, rendererClasses [&NiagaraSpriteRendererProperties]}
  {emitterName=CoreFlare, rendererClasses [&NiagaraSpriteRendererProperties]}
  {emitterName=NE_Juno_DarkGlow001, rendererClasses [&NiagaraSpriteRendererProperties]}
  {bEnabled=false, emitterName=Smoke_Sphere, rendererClasses [&NiagaraSpriteRendererProperties]}
  {emitterName=CenterFlare, simTarget=GPUComputeSim, rendererClasses [&NiagaraSpriteRendererProperties]}
 ]
 userVariables [
  User.BeamWidth {type=NiagaraFloat, defaultValue=1, description=""}
  User.bIsFrontEnd {type=NiagaraBool}
  User.bIsFrontEndPreview {type=NiagaraBool}
  User.Charge {type=NiagaraFloat, defaultValue=3, description=""}
  User.Color {type=LinearColor, description="", defaultValue {struct=LinearColor, value=<rgba 0 0.161 0.5 1>}}
  User.ColorMult {type=NiagaraFloat, defaultValue=1, description=""}
  User.Duration {type=NiagaraFloat, defaultValue=3.8, description=""}
  User.ElementSizeScale {type=NiagaraFloat, defaultValue=1, description=""}
  User.GlobalOpacity {type=NiagaraFloat, defaultValue=1, description=""}
  User.GlobalScale {type=NiagaraFloat, defaultValue=0.75, description=""}
  User.ImpactActive {type=NiagaraFloat, defaultValue=0, description=""}
  User.ImpactOffset {type=NiagaraFloat, defaultValue=0, description=""}
  User.ImpactScale {type=NiagaraFloat, defaultValue=2, description=""}
  User.IsWater {type=NiagaraBool}
  User.PostNoiseStrength {type=NiagaraFloat, defaultValue=1, description=""}
  User.PostSizeScale {type=NiagaraFloat, defaultValue=1, description=""}
  User.PostStrength {type=NiagaraFloat, defaultValue=1, description=""}
  User.PrimaryColor {type=LinearColor, description="", defaultValue {struct=LinearColor, value=<rgba 0.755 3.9375 5 1>}}
  User.Scaling_SpawnRadius {type=NiagaraFloat, defaultValue=1, description=""}
  User.SecondaryColor {type=LinearColor, description="", defaultValue {struct=LinearColor, value=<rgba 0.13 0.3272 1 3>}}
  User.VehicleHasDriver {type=NiagaraBool, defaultValue=true}
  User.WaterPlaneLocation {type=NiagaraPosition, description="", defaultValue {}}
  User.WidthMult {type=NiagaraFloat, defaultValue=1, description=""}
 ]
}
